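# Niagara effect for the Clash emote that wraps a character in flowing mesh particles and ribbons along the skeleton.
# FN_BoneFlow places particles on splines through filtered bone chains, using normalized spawn group for height, angle and radius.
# Limb and body meshes compute ring rotations from spline tangents and animate scale and material parameters, with delays staggering legs, body, then arms.
# Back and foot-to-hand ribbons add random offsets, varying widths and curl noise.
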NiagaraSystem /BRCosmetics/Athena/Effects/Particles/Emotes/NS_Emote_Clash/NS_Emote_Clash
dependencies {
 usedDataInterfaces [&NiagaraDataInterfaceCurve, &NiagaraDataInterfaceSkeletalMesh, &NiagaraDataInterfaceArrayFloat3]
 usedDynamicInputs []
 usedModules [&SystemState]
}
emitters {
 BoneFlow_Arm_L {
  emitterUpdateScript [
   &EmitterState(Life Cycle Mode=::NewEnumerator1, Loop Behavior=::NewEnumerator1, Loop Delay=0.6, Loop Duration=0.6, UseLoopDelay=true)
   &SpawnRate(SpawnRate=34, Use Spawn Probability=false)
  ]
  particleSpawnScript [
   &FN_SetSpawnGroup(Inclusivity=true, Normalize Spawn Group=true, Particles Per Group=20, Pattern=::NewEnumerator0, Set Ribbon ID and Ribbon Link Order=true) {module=FN_SetSpawnGroup001}
   &InitializeParticle(
    Color=<rgba 0.9249 1 0.9295 1>
    Lifetime=&FloatFromCurve(
     FloatCurve=DataInterface() {propertyValues {Curve {keys [<at 0 0.5>,<at 0.7 0.7>,<at 0.9 1.2>]}}}
     CurveIndex=&RandomRangeFloat(Minimum=0, Maximum=1)
    )
    Mesh Scale Max=<0.9 0.9 2.5>
    Mesh Scale Min=<0.6 0.6 2>
    Mesh Scale Mode=::NewEnumerator2
    UsePositionOffset=false
    Write Material Random=true
   )
   &InitialMeshOrientation(Mesh Orientation Mode=::NewEnumerator2, Use Rotation Vector=true)
   &CameraOffset(Camera Offset Amount=10)
  ]
  particleUpdateScript [
   &ParticleState()
   /Game/Effects/Niagara/Modules/BoneFlow/FN_BoneFlow_v4/FN_BoneFlow(
    Angle:NiagaraFloat=&Multiply_Float(A=$Particles.Normalized Spawn Group, B=2)
    BoneID:NiagaraInt32=&ModuloInt(A=$Particles.SpawnGroup, B=1)
    Height:NiagaraFloat=&Multiply_Float(A=$Particles.Normalized Spawn Group, B=1.05)
    Interpolate Mesh Sampling=false
    Radius:NiagaraFloat=&Multiply_Float(
     A=&FloatFromCurve(
      FloatCurve=DataInterface() {propertyValues {Curve {keys [
          {arriveTangent=64.149, interpMode=RCIM_Cubic, leaveTangent=64.149, tangentMode=RCTM_User, time=0, value=0}
          <at 0.5 10 RCIM_Cubic>
          {arriveTangent=-53.2228, interpMode=RCIM_Cubic, leaveTangent=-53.2228, tangentMode=RCTM_User, time=1, value=0}
         ]}}}
      CurveIndex=$Particles.Normalized Spawn Group
     )
     B=0.5
    )
    SideVector2:Vector3f=<0 0 0>
    SideVector3:Vector3f=<1 1 1>
    SkeletalMesh=DataInterface() {propertyValues {
      DefaultMesh=null
      ExcludeBoneName=root
      OptionalComponentClass=null
      PreviewMesh=/BRCosmetics/Characters/Player/Male/Medium/Bodies/M_MED_Clash_V_Casual/Meshes/M_MED_Clash_V_Casual
      SkinningMode=SkinOnTheFly
      SoftSourceActor=null
      SourceMode=Default
      UvSetIndex=0
      WholeMeshLOD=-1
      bExcludeBone=false
      bReadDeformedGeometry=true
      bRequireCurrentFrameData=true
      ComponentTags []
      FilteredBones [upperarm_l,lowerarm_l,hand_l]
      FilteredSockets []
      MeshUserParameter {parameter {name=None, typeDefHandle {registeredTypeIndex=1}, varData []}}
      SamplingRegions []
     }}
    Stabalize Twisting=true
    Write Position Attribute=true
   ) {module=FN_BoneFlow001}
   /SetVariables(Particles.BoneFlowTangent:Vector3f=$Output.FN_BoneFlow001.SplineTangent) {isSetParameters=true, module=SetVariables_#1}
   /SetVariables(Particles.RingRotation:Quat4f=&MultiplyQuaternion(
     Quaternion A=&MakeQuatFromAxisAngle(Axis=$Particles.BoneFlowTangent, AngleInDegrees=&RandomRangeFloat(Minimum=0, Maximum=360))
     Quaternion B=&FindQuatBetweenTwoVectors(From Vector=<0.4 0.4 1>, To Vector=$Particles.BoneFlowTangent)
    )) {isSetParameters=true, module=SetVariables_#2}
   /SetVariables(Particles.ScaleMeshShapeCurve=DataInterface() {propertyValues {Curve {keys [<at 0 1>,<at 1 1>]}}}) {isSetParameters=true, module=SetVariables_#3}
   /SetVariables(Particles.ScaleMeshXYCurve=DataInterface() {propertyValues {Curve {keys [
        {interpMode=RCIM_Cubic, tangentMode=RCTM_User, time=0, value=0}
        {interpMode=RCIM_Cubic, tangentMode=RCTM_User, time=0.35, value=0.7}
        {arriveTangent=-0.2106, interpMode=RCIM_Cubic, leaveTangent=-0.2106, tangentMode=RCTM_User, time=1, value=1.2}
       ]}}}) {isSetParameters=true, module=SetVariables_#4}
   &ScaleMeshSize(Scale Factor=&Multiply_Vector(
     A=&MakeVector(
      X=&FloatFromCurve(FloatCurve=$Particles.ScaleMeshXYCurve)
      Y=&FloatFromCurve(FloatCurve=$Particles.ScaleMeshXYCurve)
      Z=&FloatFromCurve(FloatCurve=DataInterface() {propertyValues {Curve {keys [<at 0 1>, {interpMode=RCIM_Cubic, tangentMode=RCTM_User, time=0.6, value=1}, {interpMode=RCIM_Cubic, tangentMode=RCTM_User, time=1, value=0}]}}})
     )
     B=&MakeVector(
      X=&FloatFromCurve(FloatCurve=$Particles.ScaleMeshShapeCurve, CurveIndex=$Particles.Normalized Spawn Group)
      Y=&FloatFromCurve(FloatCurve=$Particles.ScaleMeshShapeCurve, CurveIndex=$Particles.Normalized Spawn Group)
      Z=1
     )
    ))
   &DynamicMaterialParameters(
    Index 0 Param 1=0.5
    Index 0 Param 2=&FloatFromCurve(FloatCurve=DataInterface() {propertyValues {Curve {keys [
         {interpMode=RCIM_Cubic, tangentMode=RCTM_User, time=0, value=0.5}
         {arriveTangent=2.7221, interpMode=RCIM_Cubic, leaveTangent=2.7221, tangentMode=RCTM_User, time=0.5, value=1.5}
         {arriveTangent=0.4617, interpMode=RCIM_Cubic, leaveTangent=0.4617, tangentMode=RCTM_User, time=0.95, value=2}
         {interpMode=RCIM_Cubic, tangentMode=RCTM_User, time=1, value=2.5}
        ]}}})
    Index 1 Param 1=&FloatFromCurve(FloatCurve=DataInterface() {propertyValues {Curve {keys [<at 0 2>,<at 0.4 0.5>]}}})
    Index 1 Param 2=2
    Index 1 Param 3=&FloatFromCurve(FloatCurve=DataInterface() {propertyValues {Curve {keys [<at 0 10>,<at 0.3 1>]}}})
    Module.Param0WriteEnabled=true
    Module.Param0WriteEnabled1=true
    Module.Param1WriteEnabled=true
    Module.Param1WriteEnabled1=true
    Module.Param2WriteEnabled=true
    Module.Param2WriteEnabled1=true
    Module.Param3WriteEnabled=true
    Module.Param3WriteEnabled1=true
    Write Parameter Index 1=true
   )
  ]
  rendererClasses [&NiagaraMeshRendererProperties]
  renderers [{rendererClass=&NiagaraMeshRendererProperties, rendererIndex=0}]
 }
 BoneFlow_Arm_R {
  emitterUpdateScript [
   &EmitterState(Life Cycle Mode=::NewEnumerator1, Loop Behavior=::NewEnumerator1, Loop Delay=0.6, Loop Duration=0.6, UseLoopDelay=true)
   &SpawnRate(SpawnRate=34, Use Spawn Probability=false)
  ]
  particleSpawnScript [
   &FN_SetSpawnGroup(Inclusivity=true, Normalize Spawn Group=true, Particles Per Group=20, Pattern=::NewEnumerator0, Set Ribbon ID and Ribbon Link Order=true) {module=FN_SetSpawnGroup001}
   &InitializeParticle(
    Color=<rgba 0.9249 1 0.9295 1>
    Lifetime=&FloatFromCurve(
     FloatCurve=DataInterface() {propertyValues {Curve {keys [<at 0 0.5>,<at 0.7 0.7>,<at 0.9 1.2>]}}}
     CurveIndex=&RandomRangeFloat(Minimum=0, Maximum=1)
    )
    Mesh Scale Max=<0.9 0.9 2.5>
    Mesh Scale Min=<0.6 0.6 2>
    Mesh Scale Mode=::NewEnumerator2
    UsePositionOffset=false
    Write Material Random=true
   )
   &InitialMeshOrientation(Mesh Orientation Mode=::NewEnumerator2, Use Rotation Vector=true)
   &CameraOffset(Camera Offset Amount=10)
  ]
  particleUpdateScript [
   &ParticleState()
   &FN_BoneFlow(
    Angle:NiagaraFloat=&Multiply_Float(A=$Particles.Normalized Spawn Group, B=2)
    BoneID:NiagaraInt32=&ModuloInt(A=$Particles.SpawnGroup, B=1)
    Height:NiagaraFloat=&Multiply_Float(A=$Particles.Normalized Spawn Group, B=1.05)
    Interpolate Mesh Sampling=false
    Radius:NiagaraFloat=&Multiply_Float(
     A=&FloatFromCurve(
      FloatCurve=DataInterface() {propertyValues {Curve {keys [
          {arriveTangent=64.149, interpMode=RCIM_Cubic, leaveTangent=64.149, tangentMode=RCTM_User, time=0, value=0}
          <at 0.5 10 RCIM_Cubic>
          {arriveTangent=-53.2228, interpMode=RCIM_Cubic, leaveTangent=-53.2228, tangentMode=RCTM_User, time=1, value=0}
         ]}}}
      CurveIndex=$Particles.Normalized Spawn Group
     )
     B=0.5
    )
    SideVector2:Vector3f=<0 0 0>
    SideVector3:Vector3f=<1 1 1>
    SkeletalMesh=DataInterface() {propertyValues {
      DefaultMesh=null
      ExcludeBoneName=root
      OptionalComponentClass=null
      PreviewMesh=&M_MED_Clash_V_Casual
      SkinningMode=SkinOnTheFly
      SoftSourceActor=null
      SourceMode=Default
      UvSetIndex=0
      WholeMeshLOD=-1
      bExcludeBone=false
      bReadDeformedGeometry=true
      bRequireCurrentFrameData=true
      ComponentTags []
      FilteredBones [upperarm_r,lowerarm_r,hand_r]
      FilteredSockets []
      MeshUserParameter {parameter {name=None, typeDefHandle {registeredTypeIndex=1}, varData []}}
      SamplingRegions []
     }}
    Stabalize Twisting=true
    Write Position Attribute=true
   ) {module=FN_BoneFlow001}
   /SetVariables(Particles.BoneFlowTangent:Vector3f=$Output.FN_BoneFlow001.SplineTangent) {isSetParameters=true, module=SetVariables_#1}
   /SetVariables(Particles.RingRotation:Quat4f=&MultiplyQuaternion(
     Quaternion A=&MakeQuatFromAxisAngle(Axis=$Particles.BoneFlowTangent, AngleInDegrees=&RandomRangeFloat(Minimum=0, Maximum=360))
     Quaternion B=&FindQuatBetweenTwoVectors(From Vector=<0.4 0.4 1>, To Vector=$Particles.BoneFlowTangent)
    )) {isSetParameters=true, module=SetVariables_#2}
   /SetVariables(Particles.ScaleMeshShapeCurve=DataInterface() {propertyValues {Curve {keys [<at 0 1>,<at 1 1>]}}}) {isSetParameters=true, module=SetVariables_#3}
   /SetVariables(Particles.ScaleMeshXYCurve=DataInterface() {propertyValues {Curve {keys [
        {interpMode=RCIM_Cubic, tangentMode=RCTM_User, time=0, value=0}
        {interpMode=RCIM_Cubic, tangentMode=RCTM_User, time=0.35, value=0.7}
        {arriveTangent=-0.2106, interpMode=RCIM_Cubic, leaveTangent=-0.2106, tangentMode=RCTM_User, time=1, value=1.2}
       ]}}}) {isSetParameters=true, module=SetVariables_#4}
   &ScaleMeshSize(Scale Factor=&Multiply_Vector(
     A=&MakeVector(
      X=&FloatFromCurve(FloatCurve=$Particles.ScaleMeshXYCurve)
      Y=&FloatFromCurve(FloatCurve=$Particles.ScaleMeshXYCurve)
      Z=&FloatFromCurve(FloatCurve=DataInterface() {propertyValues {Curve {keys [<at 0 1>, {interpMode=RCIM_Cubic, tangentMode=RCTM_User, time=0.6, value=1}, {interpMode=RCIM_Cubic, tangentMode=RCTM_User, time=1, value=0}]}}})
     )
     B=&MakeVector(
      X=&FloatFromCurve(FloatCurve=$Particles.ScaleMeshShapeCurve, CurveIndex=$Particles.Normalized Spawn Group)
      Y=&FloatFromCurve(FloatCurve=$Particles.ScaleMeshShapeCurve, CurveIndex=$Particles.Normalized Spawn Group)
      Z=1
     )
    ))
   &DynamicMaterialParameters(
    Index 0 Param 1=0.5
    Index 0 Param 2=&FloatFromCurve(FloatCurve=DataInterface() {propertyValues {Curve {keys [
         {interpMode=RCIM_Cubic, tangentMode=RCTM_User, time=0, value=0.5}
         {arriveTangent=2.7221, interpMode=RCIM_Cubic, leaveTangent=2.7221, tangentMode=RCTM_User, time=0.5, value=1.5}
         {arriveTangent=0.4617, interpMode=RCIM_Cubic, leaveTangent=0.4617, tangentMode=RCTM_User, time=0.95, value=2}
         {interpMode=RCIM_Cubic, tangentMode=RCTM_User, time=1, value=2.5}
        ]}}})
    Index 1 Param 1=&FloatFromCurve(FloatCurve=DataInterface() {propertyValues {Curve {keys [<at 0 2>,<at 0.4 0.5>]}}})
    Index 1 Param 2=2
    Index 1 Param 3=&FloatFromCurve(FloatCurve=DataInterface() {propertyValues {Curve {keys [<at 0 10>,<at 0.3 1>]}}})
    Module.Param0WriteEnabled=true
    Module.Param0WriteEnabled1=true
    Module.Param1WriteEnabled=true
    Module.Param1WriteEnabled1=true
    Module.Param2WriteEnabled=true
    Module.Param2WriteEnabled1=true
    Module.Param3WriteEnabled=true
    Module.Param3WriteEnabled1=true
    Write Parameter Index 1=true
   )
  ]
  rendererClasses [&NiagaraMeshRendererProperties]
  renderers [{rendererClass=&NiagaraMeshRendererProperties, rendererIndex=0}]
 }
 BoneFlow_Back {
  emitterUpdateScript [
   &EmitterState(Life Cycle Mode=::NewEnumerator1, Loop Behavior=::NewEnumerator1, Loop Duration=0.5, UseLoopDelay=true)
   &SpawnBurst_Instantaneous(Spawn Count=10, Use Spawn Probability=false, UseLoopCountLimit=false)
   &SpawnRate(SpawnRate=25, Use Spawn Probability=false)
  ]
  particleSpawnScript [
   &FN_SetSpawnGroup(Inclusivity=true, Normalize Spawn Group=true, Pattern=::NewEnumerator0, Set Ribbon ID and Ribbon Link Order=true)
   &InitializeRibbon(
    Lifetime Max=1.8
    Lifetime Min=1.2
    Lifetime Mode:ENiagara_LifetimeMode=::NewEnumerator1
    Ribbon Width=&FloatFromCurve(
     FloatCurve=DataInterface() {propertyValues {Curve {keys [<at 0 0.3>,<at 0.7 0.5>,<at 0.8 3>,<at 1 3.5>]}}}
     CurveIndex=&RandomRangeFloat(Minimum=0, Maximum=1, Random Seed=5)
    )
    UsePositionOffset=false
    Write Material Random=true
   )
   &CameraOffset(Camera Offset Amount=10) {enabled=false}
  ]
  particleUpdateScript [
   &ParticleState()
   /Game/Effects/Niagara/Modules/BoneFlow/FN_BoneFlow(
    Angle:NiagaraFloat=&Multiply_Float(A=&RandomRangeFloat(Minimum=0.3, Maximum=0.4), B=1)
    BoneID:NiagaraInt32=&ModuloInt(A=$Particles.SpawnGroup, B=1)
    Enable Position Offsets=true
    Height:NiagaraFloat=&Multiply_Float(
     A=&Multiply_Float(
      A=$Particles.Normalized Spawn Group
      B=&FloatFromCurve(FloatCurve=DataInterface() {propertyValues {Curve {keys [{arriveTangent=1.927, interpMode=RCIM_Cubic, leaveTangent=1.927, tangentMode=RCTM_User, time=0, value=0.6}, <at 0.2 0.9>, <at 0.5 1>]}}})
     )
     B=&RandomRangeFloat(Minimum=0.8, Maximum=1.1)
    )
    Offset End:Vector3f=&RandomRangeVector(Minimum=<0 0 0>, Maximum=<-5 0 0>)
    Offset Start:Vector3f=&RandomRangeVector(Minimum=<0 0 30>, Maximum=<0 0 -30>)
    Radius:NiagaraFloat=&Multiply_Float(
     A=&FloatFromCurve(
      FloatCurve=DataInterface() {propertyValues {Curve {keys [
          {arriveTangent=47.3486, interpMode=RCIM_Cubic, leaveTangent=47.3486, tangentMode=RCTM_User, time=0, value=0}
          {arriveTangent=-31.9851, interpMode=RCIM_Cubic, leaveTangent=-31.9851, tangentMode=RCTM_User, time=0.2438, value=20.4435}
          {arriveTangent=-4.8797, interpMode=RCIM_Cubic, leaveTangent=-4.8797, tangentMode=RCTM_User, time=0.5, value=11.25}
          {arriveTangent=-30.8897, interpMode=RCIM_Cubic, leaveTangent=-30.8897, tangentMode=RCTM_User, time=0.8, value=17}
          {arriveTangent=-76.0513, interpMode=RCIM_Cubic, leaveTangent=-76.0513, tangentMode=RCTM_User, time=1, value=6.25}
         ]}}}
      CurveIndex=$Particles.Normalized Spawn Group
      Scale Curve=&FloatFromCurve(FloatCurve=DataInterface() {propertyValues {Curve {keys [{interpMode=RCIM_Cubic, tangentMode=RCTM_User, time=0.6, value=1}, {interpMode=RCIM_Cubic, tangentMode=RCTM_User, time=1, value=0}]}}})
     )
     B=&RandomRangeFloat(Minimum=1.5, Maximum=1.2)
    )
    SideVector1:Vector3f=<1 1 1>
    SideVector2:Vector3f=<1 1 1>
    SideVector3:Vector3f=<1 1 1>
    SkeletalMesh=DataInterface() {propertyValues {
      DefaultMesh=null
      ExcludeBoneName=root
      OptionalComponentClass=null
      PreviewMesh=&M_MED_Clash_V_Casual
      SkinningMode=SkinOnTheFly
      SoftSourceActor=null
      SourceMode=Default
      UvSetIndex=0
      WholeMeshLOD=-1
      bExcludeBone=false
      bReadDeformedGeometry=true
      bRequireCurrentFrameData=true
      ComponentTags []
      FilteredBones [pelvis,spine_03,head]
      FilteredSockets []
      MeshUserParameter {parameter {name=None, typeDefHandle {registeredTypeIndex=1}, varData []}}
      SamplingRegions []
     }}
    Stabalize Twisting=true
   )
   &ScaleRibbonWidth(Ribbon Width Scale=&FloatFromCurve(
     FloatCurve=DataInterface() {propertyValues {Curve {keys [
         {interpMode=RCIM_Cubic, tangentMode=RCTM_User, time=0, value=1.8}
         {interpMode=RCIM_Cubic, tangentMode=RCTM_User, time=0.2, value=0.375}
         {arriveTangent=-1.5625, interpMode=RCIM_Cubic, leaveTangent=-1.5625, time=0.4, value=2.125}
         {interpMode=RCIM_Cubic, tangentMode=RCTM_User, time=0.6, value=-0.25}
         <at 0.8 1.3>
         {arriveTangent=-5.0334, interpMode=RCIM_Cubic, leaveTangent=-5.0334, tangentMode=RCTM_User, time=1, value=0}
        ]}}}
     CurveIndex=$Particles.Normalized Spawn Group
     Scale Curve=&FloatFromCurve(FloatCurve=DataInterface() {propertyValues {Curve {keys [
          {interpMode=RCIM_Cubic, tangentMode=RCTM_User, time=0, value=5}
          {interpMode=RCIM_Cubic, tangentMode=RCTM_User, time=0.2, value=1.5}
          {interpMode=RCIM_Cubic, tangentMode=RCTM_User, time=0.85, value=0.1}
          {interpMode=RCIM_Cubic, tangentMode=RCTM_User, time=1, value=0}
         ]}}})
    ))
   &ScaleColor(
    Scale Alpha=&FloatFromCurve(FloatCurve=DataInterface() {propertyValues {Curve {keys [
         {interpMode=RCIM_Cubic, tangentMode=RCTM_User, time=0, value=0}
         <at 0.2 1>
         {interpMode=RCIM_Cubic, tangentMode=RCTM_User, time=0.8, value=1}
         {tangentMode=RCTM_User, time=1, value=0}
        ]}}})
    ScaleA=true
    ScaleRGB=true
   )
   &DynamicMaterialParameters(
    Index 0 Param 1=&FloatFromCurve(FloatCurve=DataInterface() {propertyValues {Curve {keys [
         {arriveTangent=1.8932, interpMode=RCIM_Cubic, leaveTangent=1.8932, tangentMode=RCTM_User, time=0, value=0}
         {arriveTangent=2.4517, interpMode=RCIM_Cubic, leaveTangent=2.4517, tangentMode=RCTM_User, time=0.8, value=0.85}
         <at 1 1>
        ]}}})
    Index 1 Param 1=&RandomRangeFloat(Minimum=2, Maximum=1)
    Index 1 Param 2=-1
    Index 1 Param 3=&RandomRangeFloat(Minimum=0.8, Maximum=1)
    Index 1 Param 4=&RandomRangeFloat(Minimum=0.5, Maximum=0)
    Module.Param0WriteEnabled=true
    Module.Param0WriteEnabled1=true
    Module.Param1WriteEnabled=true
    Module.Param1WriteEnabled1=true
    Module.Param2WriteEnabled=true
    Module.Param2WriteEnabled1=true
    Module.Param3WriteEnabled=true
    Module.Param3WriteEnabled1=true
    Write Parameter Index 1=true
   )
   &CurlNoiseForce(Noise Frequency=5000, Noise Strength=300, Pan Noise=true, Pan Noise Field=<0 0 2>)
   &SolveForcesAndVelocity(Clamp Velocity=false, Limit Acceleration=false)
  ]
  rendererClasses [&NiagaraRibbonRendererProperties]
  renderers [{rendererClass=&NiagaraRibbonRendererProperties}]
 }
 BoneFlow_Body {
  emitterUpdateScript [
   &EmitterState(Life Cycle Mode=::NewEnumerator1, Loop Behavior=::NewEnumerator1, Loop Delay=0.36, Loop Duration=0.65, UseLoopDelay=true)
   &SpawnRate(SpawnRate=22, Use Spawn Probability=false)
  ]
  particleSpawnScript [
   &FN_SetSpawnGroup(Inclusivity=true, Normalize Spawn Group=true, Particles Per Group=15, Pattern=::NewEnumerator0, Set Ribbon ID and Ribbon Link Order=true) {module=FN_SetSpawnGroup001}
   &InitializeParticle(
    Color=<rgba 0.9249 1 0.9295 1>
    Lifetime=&FloatFromCurve(
     FloatCurve=DataInterface() {propertyValues {Curve {keys [<at 0 0.4>,<at 0.8 0.5>,<at 1 1.2>]}}}
     CurveIndex=&RandomRangeFloat(Minimum=0, Maximum=1)
    )
    Mesh Scale Max=<3 3 8>
    Mesh Scale Min=<2.5 2.5 9>
    Mesh Scale Mode=::NewEnumerator2
    UsePositionOffset=false
    Write Material Random=true
   )
   &InitialMeshOrientation(Mesh Orientation Mode=::NewEnumerator2, Use Rotation Vector=true)
   &CameraOffset(Camera Offset Amount=10)
  ]
  particleUpdateScript [
   &ParticleState()
   /Game/Effects/Niagara/Modules/BoneFlow/FN_BoneFlow_v4/FN_BoneFlow(
    Angle:NiagaraFloat=&Multiply_Float(A=$Particles.Normalized Spawn Group, B=2)
    BoneID:NiagaraInt32=&ModuloInt(A=$Particles.SpawnGroup, B=1)
    Height:NiagaraFloat=&Multiply_Float(A=$Particles.Normalized Spawn Group, B=1.1)
    Interpolate Mesh Sampling=false
    Radius:NiagaraFloat=&Multiply_Float(
     A=&FloatFromCurve(
      FloatCurve=DataInterface() {propertyValues {Curve {keys [
          {arriveTangent=64.149, interpMode=RCIM_Cubic, leaveTangent=64.149, tangentMode=RCTM_User, time=0, value=0}
          <at 0.5 10 RCIM_Cubic>
          {arriveTangent=0.3152, interpMode=RCIM_Cubic, leaveTangent=0.3152, tangentMode=RCTM_User, time=0.8, value=0}
         ]}}}
      CurveIndex=$Particles.Normalized Spawn Group
     )
     B=1
    )
    SideVector2:Vector3f=<0 0 0>
    SideVector3:Vector3f=<1 1 1>
    SkeletalMesh=DataInterface() {propertyValues {
      DefaultMesh=null
      ExcludeBoneName=root
      OptionalComponentClass=null
      PreviewMesh=&M_MED_Clash_V_Casual
      SkinningMode=SkinOnTheFly
      SoftSourceActor=null
      SourceMode=Default
      UvSetIndex=0
      WholeMeshLOD=-1
      bExcludeBone=false
      bReadDeformedGeometry=true
      bRequireCurrentFrameData=true
      ComponentTags []
      FilteredBones [pelvis,spine_03,head]
      FilteredSockets []
      MeshUserParameter {parameter {name=None, typeDefHandle {registeredTypeIndex=1}, varData []}}
      SamplingRegions []
     }}
    Stabalize Twisting=true
    Write Position Attribute=true
   ) {module=FN_BoneFlow001}
   /SetVariables(Particles.BoneFlowTangent:Vector3f=$Output.FN_BoneFlow001.SplineTangent) {isSetParameters=true, module=SetVariables_#1}
   /SetVariables(Particles.RingRotation:Quat4f=&MultiplyQuaternion(
     Quaternion A=&MakeQuatFromAxisAngle(Axis=$Particles.BoneFlowTangent, AngleInDegrees=&RandomRangeFloat(Minimum=0, Maximum=360))
     Quaternion B=&FindQuatBetweenTwoVectors(From Vector=<0.2 0.2 1>, To Vector=$Particles.BoneFlowTangent)
    )) {isSetParameters=true, module=SetVariables_#2}
   /SetVariables(Particles.ScaleMeshShapeCurve=DataInterface() {propertyValues {Curve {keys [<at 0 1.2>,<at 0.7 1.1>,<at 0.77 0.5>]}}}) {isSetParameters=true, module=SetVariables_#3}
   /SetVariables(Particles.ScaleMeshXYCurve=DataInterface() {propertyValues {Curve {keys [
        {interpMode=RCIM_Cubic, tangentMode=RCTM_User, time=0, value=0}
        {interpMode=RCIM_Cubic, tangentMode=RCTM_User, time=0.35, value=0.7}
        {arriveTangent=-0.2106, interpMode=RCIM_Cubic, leaveTangent=-0.2106, tangentMode=RCTM_User, time=1, value=1}
       ]}}}) {isSetParameters=true, module=SetVariables_#4}
   &ScaleMeshSize(Scale Factor=&Multiply_Vector(
     A=&MakeVector(
      X=&FloatFromCurve(FloatCurve=$Particles.ScaleMeshXYCurve)
      Y=&FloatFromCurve(FloatCurve=$Particles.ScaleMeshXYCurve)
      Z=&FloatFromCurve(FloatCurve=DataInterface() {propertyValues {Curve {keys [<at 0 1>, {interpMode=RCIM_Cubic, tangentMode=RCTM_User, time=0.5, value=1}, {interpMode=RCIM_Cubic, tangentMode=RCTM_User, time=1, value=0}]}}})
     )
     B=&MakeVector(
      X=&FloatFromCurve(FloatCurve=$Particles.ScaleMeshShapeCurve, CurveIndex=$Particles.Normalized Spawn Group)
      Y=&FloatFromCurve(FloatCurve=$Particles.ScaleMeshShapeCurve, CurveIndex=$Particles.Normalized Spawn Group)
      Z=&FloatFromCurve(
       FloatCurve=DataInterface() {propertyValues {Curve {keys [
           <at 0 1>
           {interpMode=RCIM_Cubic, tangentMode=RCTM_User, time=0.6, value=0.8}
           {interpMode=RCIM_Cubic, tangentMode=RCTM_User, time=0.7, value=0.2}
          ]}}}
       CurveIndex=$Particles.Normalized Spawn Group
      )
     )
    ))
   &DynamicMaterialParameters(
    Index 0 Param 1=0.5
    Index 0 Param 2=&FloatFromCurve(FloatCurve=DataInterface() {propertyValues {Curve {keys [
         {interpMode=RCIM_Cubic, tangentMode=RCTM_User, time=0, value=0.5}
         {arriveTangent=2.7221, interpMode=RCIM_Cubic, leaveTangent=2.7221, tangentMode=RCTM_User, time=0.5, value=1.5}
         {arriveTangent=0.4617, interpMode=RCIM_Cubic, leaveTangent=0.4617, tangentMode=RCTM_User, time=0.95, value=2}
         {interpMode=RCIM_Cubic, tangentMode=RCTM_User, time=1, value=2.5}
        ]}}})
    Index 1 Param 1=&FloatFromCurve(FloatCurve=DataInterface() {propertyValues {Curve {keys [<at 0 2>,<at 0.4 0.5>]}}})
    Index 1 Param 2=2
    Index 1 Param 3=&FloatFromCurve(FloatCurve=DataInterface() {propertyValues {Curve {keys [<at 0 0>,<at 0.1 12>,<at 0.3 1>]}}})
    Module.Param0WriteEnabled=true
    Module.Param0WriteEnabled1=true
    Module.Param1WriteEnabled=true
    Module.Param1WriteEnabled1=true
    Module.Param2WriteEnabled=true
    Module.Param2WriteEnabled1=true
    Module.Param3WriteEnabled=true
    Module.Param3WriteEnabled1=true
    Write Parameter Index 1=true
   )
  ]
  rendererClasses [&NiagaraMeshRendererProperties]
  renderers [{rendererClass=&NiagaraMeshRendererProperties, rendererIndex=0}]
 }
 BoneFlow_LHand {
  emitterUpdateScript [
   &EmitterState(Life Cycle Mode=::NewEnumerator1, Loop Behavior=::NewEnumerator1, Loop Duration=0.5, UseLoopDelay=true)
   &SpawnBurst_Instantaneous(Spawn Count=20, Use Spawn Probability=false, UseLoopCountLimit=false)
  ]
  particleSpawnScript [
   &FN_SetSpawnGroup(Inclusivity=true, Normalize Spawn Group=true, Particles Per Group=5, Pattern=::NewEnumerator0, Set Ribbon ID and Ribbon Link Order=true)
   &InitializeRibbon(
    Lifetime Min=2.4
    Lifetime Mode:ENiagara_LifetimeMode=::NewEnumerator1
    Ribbon Width=&FloatFromCurve(
     FloatCurve=DataInterface() {propertyValues {Curve {keys [<at 0 0.3>,<at 0.7 0.4>,<at 0.8 2>,<at 1 2.5>]}}}
     CurveIndex=&RandomRangeFloat(Minimum=0, Maximum=1, Random Seed=5)
    )
    UsePositionOffset=false
    Write Material Random=true
   )
   &CameraOffset(Camera Offset Amount=10) {enabled=false}
  ]
  particleUpdateScript [
   &ParticleState()
   /Game/Effects/Niagara/Modules/BoneFlow/FN_BoneFlow(
    Angle:NiagaraFloat=&Multiply_Float(A=&RandomRangeFloat(Minimum=-1, Maximum=1), B=$Particles.Normalized Spawn Group)
    BoneID:NiagaraInt32=&ModuloInt(A=$Particles.SpawnGroup, B=1)
    Enable Position Offsets=true
    Height:NiagaraFloat=&Multiply_Float(
     A=&Multiply_Float(
      A=$Particles.Normalized Spawn Group
      B=&FloatFromCurve(FloatCurve=DataInterface() {propertyValues {Curve {keys [{arriveTangent=1.927, interpMode=RCIM_Cubic, leaveTangent=1.927, tangentMode=RCTM_User, time=0, value=0.2}, <at 0.2 0.9>, <at 0.5 1>]}}})
     )
     B=&RandomRangeFloat(Minimum=0.8, Maximum=1.1)
    )
    Offset End:Vector3f=&RandomRangeVector(Minimum=<15 5 0>, Maximum=<15 -5 0>)
    Offset Start:Vector3f=&RandomRangeVector(Minimum=<-10 20 3>, Maximum=<-10 -10 -3>)
    Radius:NiagaraFloat=&Multiply_Float(
     A=&FloatFromCurve(
      FloatCurve=DataInterface() {propertyValues {Curve {keys [
          {arriveTangent=67.021, interpMode=RCIM_Cubic, leaveTangent=67.021, tangentMode=RCTM_User, time=0, value=0}
          <at 0.2 5>
          {interpMode=RCIM_Cubic, leaveTangent=-76.0513, time=1, value=0}
         ]}}}
      CurveIndex=$Particles.Normalized Spawn Group
     )
     B=&RandomRangeFloat(Minimum=2, Maximum=0.5)
    )
    SkeletalMesh=DataInterface() {propertyValues {
      DefaultMesh=null
      ExcludeBoneName=root
      OptionalComponentClass=null
      PreviewMesh=&M_MED_Clash_V_Casual
      SkinningMode=SkinOnTheFly
      SoftSourceActor=null
      SourceMode=Default
      UvSetIndex=0
      WholeMeshLOD=-1
      bExcludeBone=false
      bReadDeformedGeometry=true
      bRequireCurrentFrameData=true
      ComponentTags []
      FilteredBones [foot_l,hand_l]
      FilteredSockets []
      MeshUserParameter {parameter {name=None, typeDefHandle {registeredTypeIndex=1}, varData []}}
      SamplingRegions []
     }}
   )
   &ScaleRibbonWidth(Ribbon Width Scale=&FloatFromCurve(
     FloatCurve=DataInterface() {propertyValues {Curve {keys [
         {interpMode=RCIM_Cubic, tangentMode=RCTM_User, time=0, value=1.8}
         {interpMode=RCIM_Cubic, tangentMode=RCTM_User, time=0.2, value=1.625}
         {arriveTangent=-2.1875, interpMode=RCIM_Cubic, leaveTangent=-2.1875, time=0.4, value=2.125}
         {interpMode=RCIM_Cubic, tangentMode=RCTM_User, time=0.6, value=0.75}
         <at 0.8 1.3>
         {arriveTangent=-5.0334, interpMode=RCIM_Cubic, leaveTangent=-5.0334, tangentMode=RCTM_User, time=1, value=0}
        ]}}}
     CurveIndex=$Particles.Normalized Spawn Group
     Scale Curve=&FloatFromCurve(FloatCurve=DataInterface() {propertyValues {Curve {keys [
          <at 0 1>
          {interpMode=RCIM_Cubic, tangentMode=RCTM_User, time=0.7, value=1}
          {interpMode=RCIM_Cubic, tangentMode=RCTM_User, time=0.9, value=0}
         ]}}})
    ))
   &ScaleColor(
    Scale Alpha=&FloatFromCurve(FloatCurve=DataInterface() {propertyValues {Curve {keys [
         {interpMode=RCIM_Cubic, tangentMode=RCTM_User, time=0, value=0}
         <at 0.2 1>
         {interpMode=RCIM_Cubic, tangentMode=RCTM_User, time=0.8, value=1}
         {tangentMode=RCTM_User, time=1, value=0}
        ]}}})
    ScaleA=true
    ScaleRGB=true
   )
   &DynamicMaterialParameters(
    Index 0 Param 1=&FloatFromCurve(FloatCurve=DataInterface() {propertyValues {Curve {keys [
         {arriveTangent=2.4231, interpMode=RCIM_Cubic, leaveTangent=2.4231, tangentMode=RCTM_User, time=0, value=0}
         {arriveTangent=0.5882, interpMode=RCIM_Cubic, leaveTangent=0.5882, time=0.1, value=0.43}
         <at 0.5 0.5>
         {arriveTangent=-0.0482, interpMode=RCIM_Cubic, tangentMode=RCTM_Break, time=0.85, value=0.5}
         <at 1 0>
        ]}}})
    Index 0 Param 2=&RandomRangeFloat(Minimum=0, Maximum=1)
    Index 1 Param 1=&RandomRangeFloat(Minimum=1, Maximum=1)
    Index 1 Param 2=-1
    Index 1 Param 3=&FloatFromCurve(FloatCurve=DataInterface() {propertyValues {Curve {keys [{interpMode=RCIM_Cubic, tangentMode=RCTM_User, time=0.8, value=0.8}, {interpMode=RCIM_Cubic, tangentMode=RCTM_User, time=0.9, value=0}]}}})
    Index 1 Param 4=&RandomRangeFloat(Minimum=0.5, Maximum=0)
    Module.Param0WriteEnabled=true
    Module.Param0WriteEnabled1=true
    Module.Param1WriteEnabled=true
    Module.Param1WriteEnabled1=true
    Module.Param2WriteEnabled=true
    Module.Param2WriteEnabled1=true
    Module.Param3WriteEnabled=true
    Module.Param3WriteEnabled1=true
    Write Parameter Index 1=true
   )
   &CurlNoiseForce(Noise Frequency=5000, Noise Strength=100, Pan Noise=true, Pan Noise Field=<0 0 2>)
   &SolveForcesAndVelocity(Clamp Velocity=false, Limit Acceleration=false)
  ]
  rendererClasses [&NiagaraRibbonRendererProperties]
  renderers [{rendererClass=&NiagaraRibbonRendererProperties}]
 }
 BoneFlow_Leg_L {
  emitterUpdateScript [
   &EmitterState(Life Cycle Mode=::NewEnumerator1, Loop Behavior=::NewEnumerator1, Loop Duration=0.5, UseLoopDelay=false)
   &SpawnRate(SpawnRate=50, Use Spawn Probability=false)
  ]
  particleSpawnScript [
   &FN_SetSpawnGroup(Inclusivity=true, Normalize Spawn Group=true, Particles Per Group=25, Pattern=::NewEnumerator0, Set Ribbon ID and Ribbon Link Order=true) {module=FN_SetSpawnGroup001}
   &InitializeParticle(
    Color=<rgba 0.9249 1 0.9295 1>
    Lifetime=&FloatFromCurve(
     FloatCurve=DataInterface() {propertyValues {Curve {keys [<at 0 0.3>,<at 0.7 0.6>,<at 1 1.2>]}}}
     CurveIndex=&RandomRangeFloat(Minimum=0, Maximum=1)
    )
    Mesh Scale Max=<1.6 1.6 3>
    Mesh Scale Min=<0.9 0.9 2>
    Mesh Scale Mode=::NewEnumerator2
    UsePositionOffset=false
    Write Material Random=true
   )
   &InitialMeshOrientation(Mesh Orientation Mode=::NewEnumerator2, Use Rotation Vector=true)
   &CameraOffset(Camera Offset Amount=10)
  ]
  particleUpdateScript [
   &ParticleState()
   /Game/Effects/Niagara/Modules/BoneFlow/FN_BoneFlow_v4/FN_BoneFlow(
    Angle:NiagaraFloat=&Multiply_Float(A=$Particles.Normalized Spawn Group, B=2)
    BoneID:NiagaraInt32=&ModuloInt(A=$Particles.SpawnGroup, B=1)
    Height:NiagaraFloat=&Multiply_Float(A=$Particles.Normalized Spawn Group, B=1.1)
    Interpolate Mesh Sampling=false
    Radius:NiagaraFloat=&Multiply_Float(
     A=&FloatFromCurve(
      FloatCurve=DataInterface() {propertyValues {Curve {keys [
          {arriveTangent=64.149, interpMode=RCIM_Cubic, leaveTangent=64.149, tangentMode=RCTM_User, time=0, value=0}
          <at 0.5 10 RCIM_Cubic>
          {arriveTangent=-53.2228, interpMode=RCIM_Cubic, leaveTangent=-53.2228, tangentMode=RCTM_User, time=1, value=0}
         ]}}}
      CurveIndex=$Particles.Normalized Spawn Group
     )
     B=1
    )
    SideVector2:Vector3f=<0 0 0>
    SideVector3:Vector3f=<1 1 1>
    SkeletalMesh=DataInterface() {propertyValues {
      DefaultMesh=null
      ExcludeBoneName=root
      OptionalComponentClass=null
      PreviewMesh=&M_MED_Clash_V_Casual
      SkinningMode=SkinOnTheFly
      SoftSourceActor=null
      SourceMode=Default
      UvSetIndex=0
      WholeMeshLOD=-1
      bExcludeBone=false
      bReadDeformedGeometry=true
      bRequireCurrentFrameData=true
      ComponentTags []
      FilteredBones [foot_l,calf_l,thigh_l]
      FilteredSockets []
      MeshUserParameter {parameter {name=None, typeDefHandle {registeredTypeIndex=1}, varData []}}
      SamplingRegions []
     }}
    Stabalize Twisting=true
    Write Position Attribute=true
   ) {module=FN_BoneFlow001}
   /SetVariables(Particles.BoneFlowTangent:Vector3f=$Output.FN_BoneFlow001.SplineTangent) {isSetParameters=true, module=SetVariables_#1}
   /SetVariables(Particles.RingRotation:Quat4f=&MultiplyQuaternion(
     Quaternion A=&MakeQuatFromAxisAngle(Axis=$Particles.BoneFlowTangent, AngleInDegrees=&RandomRangeFloat(Minimum=0, Maximum=360))
     Quaternion B=&FindQuatBetweenTwoVectors(From Vector=<0.4 0.4 1>, To Vector=$Particles.BoneFlowTangent)
    )) {isSetParameters=true, module=SetVariables_#2}
   /SetVariables(Particles.ScaleMeshShapeCurve=DataInterface() {propertyValues {Curve {keys [<at 0.1 0.5>,<at 0.2 0.7>,<at 0.7 1>,<at 1 1.2>]}}}) {isSetParameters=true, module=SetVariables_#3}
   /SetVariables(Particles.ScaleMeshXYCurve=DataInterface() {propertyValues {Curve {keys [
        {interpMode=RCIM_Cubic, tangentMode=RCTM_User, time=0, value=0}
        {interpMode=RCIM_Cubic, tangentMode=RCTM_User, time=0.35, value=0.7}
        {arriveTangent=-0.2106, interpMode=RCIM_Cubic, leaveTangent=-0.2106, tangentMode=RCTM_User, time=1, value=1.2}
       ]}}}) {isSetParameters=true, module=SetVariables_#4}
   &ScaleMeshSize(Scale Factor=&Multiply_Vector(
     A=&MakeVector(
      X=&FloatFromCurve(FloatCurve=$Particles.ScaleMeshXYCurve)
      Y=&FloatFromCurve(FloatCurve=$Particles.ScaleMeshXYCurve)
      Z=&FloatFromCurve(FloatCurve=DataInterface() {propertyValues {Curve {keys [<at 0 1>, {interpMode=RCIM_Cubic, tangentMode=RCTM_User, time=0.6, value=1}, {interpMode=RCIM_Cubic, tangentMode=RCTM_User, time=1, value=0}]}}})
     )
     B=&MakeVector(
      X=&FloatFromCurve(FloatCurve=$Particles.ScaleMeshShapeCurve, CurveIndex=$Particles.Normalized Spawn Group)
      Y=&FloatFromCurve(FloatCurve=$Particles.ScaleMeshShapeCurve, CurveIndex=$Particles.Normalized Spawn Group)
      Z=1
     )
    ))
   &DynamicMaterialParameters(
    Index 0 Param 1=0.5
    Index 0 Param 2=&FloatFromCurve(FloatCurve=DataInterface() {propertyValues {Curve {keys [
         {interpMode=RCIM_Cubic, tangentMode=RCTM_User, time=0, value=0.5}
         {arriveTangent=2.7221, interpMode=RCIM_Cubic, leaveTangent=2.7221, tangentMode=RCTM_User, time=0.5, value=1.5}
         {arriveTangent=0.4617, interpMode=RCIM_Cubic, leaveTangent=0.4617, tangentMode=RCTM_User, time=0.95, value=2}
         {interpMode=RCIM_Cubic, tangentMode=RCTM_User, time=1, value=2.5}
        ]}}})
    Index 1 Param 1=&FloatFromCurve(FloatCurve=DataInterface() {propertyValues {Curve {keys [<at 0 2>,<at 0.4 0.5>]}}})
    Index 1 Param 2=2
    Index 1 Param 3=&FloatFromCurve(FloatCurve=DataInterface() {propertyValues {Curve {keys [<at 0 12>,<at 0.3 1>]}}})
    Module.Param0WriteEnabled=true
    Module.Param0WriteEnabled1=true
    Module.Param1WriteEnabled=true
    Module.Param1WriteEnabled1=true
    Module.Param2WriteEnabled=true
    Module.Param2WriteEnabled1=true
    Module.Param3WriteEnabled=true
    Module.Param3WriteEnabled1=true
    Write Parameter Index 1=true
   )
  ]
  rendererClasses [&NiagaraMeshRendererProperties]
  renderers [{rendererClass=&NiagaraMeshRendererProperties, rendererIndex=0}]
 }
 BoneFlow_Leg_R {
  emitterUpdateScript [
   &EmitterState(Life Cycle Mode=::NewEnumerator1, Loop Behavior=::NewEnumerator1, Loop Duration=0.5, UseLoopDelay=false)
   &SpawnRate(SpawnRate=50, Use Spawn Probability=false)
  ]
  particleSpawnScript [
   &FN_SetSpawnGroup(Inclusivity=true, Normalize Spawn Group=true, Particles Per Group=25, Pattern=::NewEnumerator0, Set Ribbon ID and Ribbon Link Order=true) {module=FN_SetSpawnGroup001}
   &InitializeParticle(
    Color=<rgba 0.9249 1 0.9295 1>
    Lifetime=&FloatFromCurve(
     FloatCurve=DataInterface() {propertyValues {Curve {keys [<at 0 0.3>,<at 0.7 0.6>,<at 1 1.2>]}}}
     CurveIndex=&RandomRangeFloat(Minimum=0, Maximum=1)
    )
    Mesh Scale Max=<1.6 1.6 3>
    Mesh Scale Min=<0.9 0.9 2>
    Mesh Scale Mode=::NewEnumerator2
    UsePositionOffset=false
    Write Material Random=true
   )
   &InitialMeshOrientation(Mesh Orientation Mode=::NewEnumerator2, Use Rotation Vector=true)
   &CameraOffset(Camera Offset Amount=10)
  ]
  particleUpdateScript [
   &ParticleState()
   &FN_BoneFlow(
    Angle:NiagaraFloat=&Multiply_Float(A=$Particles.Normalized Spawn Group, B=2)
    BoneID:NiagaraInt32=&ModuloInt(A=$Particles.SpawnGroup, B=1)
    Height:NiagaraFloat=&Multiply_Float(A=$Particles.Normalized Spawn Group, B=1.1)
    Interpolate Mesh Sampling=false
    Radius:NiagaraFloat=&Multiply_Float(
     A=&FloatFromCurve(
      FloatCurve=DataInterface() {propertyValues {Curve {keys [
          {arriveTangent=64.149, interpMode=RCIM_Cubic, leaveTangent=64.149, tangentMode=RCTM_User, time=0, value=0}
          <at 0.5 10 RCIM_Cubic>
          {arriveTangent=-53.2228, interpMode=RCIM_Cubic, leaveTangent=-53.2228, tangentMode=RCTM_User, time=1, value=0}
         ]}}}
      CurveIndex=$Particles.Normalized Spawn Group
     )
     B=1
    )
    SideVector2:Vector3f=<0 0 0>
    SideVector3:Vector3f=<1 1 1>
    SkeletalMesh=DataInterface() {propertyValues {
      DefaultMesh=null
      ExcludeBoneName=root
      OptionalComponentClass=null
      PreviewMesh=&M_MED_Clash_V_Casual
      SkinningMode=SkinOnTheFly
      SoftSourceActor=null
      SourceMode=Default
      UvSetIndex=0
      WholeMeshLOD=-1
      bExcludeBone=false
      bReadDeformedGeometry=true
      bRequireCurrentFrameData=true
      ComponentTags []
      FilteredBones [foot_r,calf_r,thigh_r]
      FilteredSockets []
      MeshUserParameter {parameter {name=None, typeDefHandle {registeredTypeIndex=1}, varData []}}
      SamplingRegions []
     }}
    Stabalize Twisting=true
    Write Position Attribute=true
   ) {module=FN_BoneFlow001}
   /SetVariables(Particles.BoneFlowTangent:Vector3f=$Output.FN_BoneFlow001.SplineTangent) {isSetParameters=true, module=SetVariables_#1}
   /SetVariables(Particles.RingRotation:Quat4f=&MultiplyQuaternion(
     Quaternion A=&MakeQuatFromAxisAngle(Axis=$Particles.BoneFlowTangent, AngleInDegrees=&RandomRangeFloat(Minimum=0, Maximum=360))
     Quaternion B=&FindQuatBetweenTwoVectors(From Vector=<0.4 0.4 1>, To Vector=$Particles.BoneFlowTangent)
    )) {isSetParameters=true, module=SetVariables_#2}
   /SetVariables(Particles.ScaleMeshShapeCurve=DataInterface() {propertyValues {Curve {keys [<at 0.1 0.5>,<at 0.2 0.7>,<at 0.7 1>,<at 1 1.2>]}}}) {isSetParameters=true, module=SetVariables_#3}
   /SetVariables(Particles.ScaleMeshXYCurve=DataInterface() {propertyValues {Curve {keys [
        {interpMode=RCIM_Cubic, tangentMode=RCTM_User, time=0, value=0}
        {interpMode=RCIM_Cubic, tangentMode=RCTM_User, time=0.35, value=0.7}
        {arriveTangent=-0.2106, interpMode=RCIM_Cubic, leaveTangent=-0.2106, tangentMode=RCTM_User, time=1, value=1.2}
       ]}}}) {isSetParameters=true, module=SetVariables_#4}
   &ScaleMeshSize(Scale Factor=&Multiply_Vector(
     A=&MakeVector(
      X=&FloatFromCurve(FloatCurve=$Particles.ScaleMeshXYCurve)
      Y=&FloatFromCurve(FloatCurve=$Particles.ScaleMeshXYCurve)
      Z=&FloatFromCurve(FloatCurve=DataInterface() {propertyValues {Curve {keys [<at 0 1>, {interpMode=RCIM_Cubic, tangentMode=RCTM_User, time=0.6, value=1}, {interpMode=RCIM_Cubic, tangentMode=RCTM_User, time=1, value=0}]}}})
     )
     B=&MakeVector(
      X=&FloatFromCurve(FloatCurve=$Particles.ScaleMeshShapeCurve, CurveIndex=$Particles.Normalized Spawn Group)
      Y=&FloatFromCurve(FloatCurve=$Particles.ScaleMeshShapeCurve, CurveIndex=$Particles.Normalized Spawn Group)
      Z=1
     )
    ))
   &DynamicMaterialParameters(
    Index 0 Param 1=0.5
    Index 0 Param 2=&FloatFromCurve(FloatCurve=DataInterface() {propertyValues {Curve {keys [
         {interpMode=RCIM_Cubic, tangentMode=RCTM_User, time=0, value=0.5}
         {arriveTangent=2.7221, interpMode=RCIM_Cubic, leaveTangent=2.7221, tangentMode=RCTM_User, time=0.5, value=1.5}
         {arriveTangent=0.4617, interpMode=RCIM_Cubic, leaveTangent=0.4617, tangentMode=RCTM_User, time=0.95, value=2}
         {interpMode=RCIM_Cubic, tangentMode=RCTM_User, time=1, value=2.5}
        ]}}})
    Index 1 Param 1=&FloatFromCurve(FloatCurve=DataInterface() {propertyValues {Curve {keys [<at 0 2>,<at 0.4 0.5>]}}})
    Index 1 Param 2=2
    Index 1 Param 3=&FloatFromCurve(FloatCurve=DataInterface() {propertyValues {Curve {keys [<at 0 12>,<at 0.3 1>]}}})
    Module.Param0WriteEnabled=true
    Module.Param0WriteEnabled1=true
    Module.Param1WriteEnabled=true
    Module.Param1WriteEnabled1=true
    Module.Param2WriteEnabled=true
    Module.Param2WriteEnabled1=true
    Module.Param3WriteEnabled=true
    Module.Param3WriteEnabled1=true
    Write Parameter Index 1=true
   )
  ]
  rendererClasses [&NiagaraMeshRendererProperties]
  renderers [{rendererClass=&NiagaraMeshRendererProperties, rendererIndex=0}]
 }
 BoneFlow_RHand {
  emitterUpdateScript [
   &EmitterState(Life Cycle Mode=::NewEnumerator1, Loop Behavior=::NewEnumerator1, Loop Duration=0.5, UseLoopDelay=true)
   &SpawnBurst_Instantaneous(Spawn Count=20, Use Spawn Probability=false, UseLoopCountLimit=false)
  ]
  particleSpawnScript [
   &FN_SetSpawnGroup(Inclusivity=true, Normalize Spawn Group=true, Particles Per Group=5, Pattern=::NewEnumerator0, Set Ribbon ID and Ribbon Link Order=true)
   &InitializeRibbon(
    Lifetime Min=2.4
    Lifetime Mode:ENiagara_LifetimeMode=::NewEnumerator1
    Ribbon Width=&FloatFromCurve(
     FloatCurve=DataInterface() {propertyValues {Curve {keys [<at 0 0.3>,<at 0.7 0.4>,<at 0.8 2>,<at 1 2.5>]}}}
     CurveIndex=&RandomRangeFloat(Minimum=0, Maximum=1, Random Seed=5)
    )
    UsePositionOffset=false
    Write Material Random=true
   )
   &CameraOffset(Camera Offset Amount=10) {enabled=false}
  ]
  particleUpdateScript [
   &ParticleState()
   /Game/Effects/Niagara/Modules/BoneFlow/FN_BoneFlow(
    Angle:NiagaraFloat=&Multiply_Float(A=&RandomRangeFloat(Minimum=-1, Maximum=1), B=$Particles.Normalized Spawn Group)
    BoneID:NiagaraInt32=&ModuloInt(A=$Particles.SpawnGroup, B=1)
    Enable Position Offsets=true
    Height:NiagaraFloat=&Multiply_Float(
     A=&Multiply_Float(
      A=$Particles.Normalized Spawn Group
      B=&FloatFromCurve(FloatCurve=DataInterface() {propertyValues {Curve {keys [{arriveTangent=1.927, interpMode=RCIM_Cubic, leaveTangent=1.927, tangentMode=RCTM_User, time=0, value=0.2}, <at 0.2 0.9>, <at 0.5 1>]}}})
     )
     B=&RandomRangeFloat(Minimum=0.8, Maximum=1.1)
    )
    Offset End:Vector3f=&RandomRangeVector(Minimum=<-15 5 0>, Maximum=<-15 -5 0>)
    Offset Start:Vector3f=&RandomRangeVector(Minimum=<10 10 3>, Maximum=<10 -20 -3>)
    Radius:NiagaraFloat=&Multiply_Float(
     A=&FloatFromCurve(
      FloatCurve=DataInterface() {propertyValues {Curve {keys [
          {arriveTangent=84.3767, interpMode=RCIM_Cubic, leaveTangent=84.3767, tangentMode=RCTM_User, time=0, value=0}
          <at 0.2 5>
          {interpMode=RCIM_Cubic, leaveTangent=-76.0513, time=1, value=0}
         ]}}}
      CurveIndex=$Particles.Normalized Spawn Group
     )
     B=&RandomRangeFloat(Minimum=2, Maximum=0.5)
    )
    SkeletalMesh=DataInterface() {propertyValues {
      DefaultMesh=null
      ExcludeBoneName=root
      OptionalComponentClass=null
      PreviewMesh=&M_MED_Clash_V_Casual
      SkinningMode=SkinOnTheFly
      SoftSourceActor=null
      SourceMode=Default
      UvSetIndex=0
      WholeMeshLOD=-1
      bExcludeBone=false
      bReadDeformedGeometry=true
      bRequireCurrentFrameData=true
      ComponentTags []
      FilteredBones [foot_r,hand_r]
      FilteredSockets []
      MeshUserParameter {parameter {name=None, typeDefHandle {registeredTypeIndex=1}, varData []}}
      SamplingRegions []
     }}
   )
   &ScaleRibbonWidth(Ribbon Width Scale=&FloatFromCurve(
     FloatCurve=DataInterface() {propertyValues {Curve {keys [
         {interpMode=RCIM_Cubic, tangentMode=RCTM_User, time=0, value=1.8}
         {interpMode=RCIM_Cubic, tangentMode=RCTM_User, time=0.2, value=1.625}
         {arriveTangent=-2.1875, interpMode=RCIM_Cubic, leaveTangent=-2.1875, time=0.4, value=2.125}
         {interpMode=RCIM_Cubic, tangentMode=RCTM_User, time=0.6, value=0.75}
         <at 0.8 1.3>
         {arriveTangent=-5.0334, interpMode=RCIM_Cubic, leaveTangent=-5.0334, tangentMode=RCTM_User, time=1, value=0}
        ]}}}
     CurveIndex=$Particles.Normalized Spawn Group
     Scale Curve=&FloatFromCurve(FloatCurve=DataInterface() {propertyValues {Curve {keys [
          <at 0 1>
          {interpMode=RCIM_Cubic, tangentMode=RCTM_User, time=0.7, value=1}
          {interpMode=RCIM_Cubic, tangentMode=RCTM_User, time=0.9, value=0}
         ]}}})
    ))
   &ScaleColor(
    Scale Alpha=&FloatFromCurve(FloatCurve=DataInterface() {propertyValues {Curve {keys [
         {interpMode=RCIM_Cubic, tangentMode=RCTM_User, time=0, value=0}
         <at 0.2 1>
         {interpMode=RCIM_Cubic, tangentMode=RCTM_User, time=0.8, value=1}
         {tangentMode=RCTM_User, time=1, value=0}
        ]}}})
    ScaleA=true
    ScaleRGB=true
   )
   &DynamicMaterialParameters(
    Index 0 Param 1=&FloatFromCurve(FloatCurve=DataInterface() {propertyValues {Curve {keys [
         {arriveTangent=2.4231, interpMode=RCIM_Cubic, leaveTangent=2.4231, tangentMode=RCTM_User, time=0, value=0}
         {arriveTangent=0.5882, interpMode=RCIM_Cubic, leaveTangent=0.5882, time=0.1, value=0.44}
         <at 0.5 0.5>
         {arriveTangent=-0.0482, interpMode=RCIM_Cubic, tangentMode=RCTM_Break, time=0.85, value=0.5}
         <at 1 0>
        ]}}})
    Index 0 Param 2=&RandomRangeFloat(Minimum=0, Maximum=1)
    Index 1 Param 1=&RandomRangeFloat(Minimum=1, Maximum=1)
    Index 1 Param 2=-1
    Index 1 Param 3=&FloatFromCurve(FloatCurve=DataInterface() {propertyValues {Curve {keys [{interpMode=RCIM_Cubic, tangentMode=RCTM_User, time=0.8, value=0.8}, {interpMode=RCIM_Cubic, tangentMode=RCTM_User, time=0.9, value=0}]}}})
    Index 1 Param 4=&RandomRangeFloat(Minimum=0.5, Maximum=0)
    Module.Param0WriteEnabled=true
    Module.Param0WriteEnabled1=true
    Module.Param1WriteEnabled=true
    Module.Param1WriteEnabled1=true
    Module.Param2WriteEnabled=true
    Module.Param2WriteEnabled1=true
    Module.Param3WriteEnabled=true
    Module.Param3WriteEnabled1=true
    Write Parameter Index 1=true
   )
   &CurlNoiseForce(Noise Frequency=5000, Noise Strength=100, Pan Noise=true, Pan Noise Field=<0 0 2>)
   &SolveForcesAndVelocity(Clamp Velocity=false, Limit Acceleration=false)
  ]
  rendererClasses [&NiagaraRibbonRendererProperties]
  renderers [{rendererClass=&NiagaraRibbonRendererProperties}]
 }
 Mesh_Body {
  emitterUpdateScript [
   &EmitterState(Inactive Response=::NewEnumerator1, Life Cycle Mode=::NewEnumerator1, Loop Behavior=::NewEnumerator1, Loop Duration=4, UseLoopDelay=false)
   &SpawnBurst_Instantaneous(Spawn Time=0.2, Use Spawn Probability=false, UseLoopCountLimit=false)
  ]
  particleSpawnScript [
   &FN_SetSpawnGroup(Number of Spawn Groups=2)
   &InitializeParticle(
    Lifetime=2.4
    Mesh Scale Max=<1 1.7 1.7>
    Mesh Scale Min=<1 1.5 1.5>
    Mesh Scale Mode=::NewEnumerator2
    UsePositionOffset=false
    Write Material Random=true
   )
   &InitialMeshOrientation(
    Orientation Axis=<0 0 1>
    Rotation=<0 0 0.6>
    Rotation Coordinate Space=::NewEnumerator2
    Use Orientation Vector=true
    Use Rotation Vector=true
   )
  ]
  particleUpdateScript [
   &ParticleState()
   &SkeletalMeshLocation(
    Bone / Socket Index=$Particles.SpawnGroup
    Bone Sampling Mode=::NewEnumerator4
    "Interpolate Mesh Sampling (Update)"=true
    MaskSpawnGroup=false
    Orientation Sampling=::NewEnumerator1
    Sampled Position Offset=&SelectVectorFromArray(Direct Array Index=$Particles.SpawnGroup, Vector Selection Array=DataInterface() {propertyValues {MaxElements=2, FloatData [<10 0 0>]}})
    Skeletal Mesh=DataInterface() {propertyValues {
      DefaultMesh=null
      ExcludeBoneName=root
      OptionalComponentClass=null
      PreviewMesh=&M_MED_Clash_V_Casual
      SkinningMode=SkinOnTheFly
      SoftSourceActor=null
      SourceMode=Default
      UvSetIndex=0
      WholeMeshLOD=-1
      bExcludeBone=false
      bReadDeformedGeometry=true
      bRequireCurrentFrameData=true
      ComponentTags []
      FilteredBones [spine_02]
      FilteredSockets []
      MeshUserParameter {parameter {name=None, typeDefHandle {registeredTypeIndex=1}, varData []}}
      SamplingRegions []
     }}
   )
   &UpdateMeshOrientation(Rotation Rate=0.1, Rotation Vector=<0 0 1>)
   &DynamicMaterialParameters(
    Index 0 Param 1=&FloatFromCurve(FloatCurve=DataInterface() {propertyValues {Curve {keys [<at 0 1>,<at 0.5 0.5>]}}})
    Index 0 Param 2=&FloatFromCurve(FloatCurve=DataInterface() {propertyValues {Curve {keys [
         {interpMode=RCIM_Cubic, leaveTangent=-2.9591, tangentMode=RCTM_Break, time=0, value=1.5}
         {arriveTangent=-0.3114, interpMode=RCIM_Cubic, tangentMode=RCTM_Break, time=1, value=0}
        ]}}})
    Index 1 Param 1=&FloatFromCurve(FloatCurve=DataInterface() {propertyValues {Curve {keys [
         {arriveTangent=-10.7468, interpMode=RCIM_Cubic, leaveTangent=-10.7468, tangentMode=RCTM_User, time=0, value=2}
         {interpMode=RCIM_Cubic, leaveTangent=0.5305, tangentMode=RCTM_Break, time=0.1, value=1}
         {interpMode=RCIM_Cubic, tangentMode=RCTM_Break, time=1, value=0.2}
        ]}}})
    Index 1 Param 2=2
    Index 1 Param 3=&FloatFromCurve(FloatCurve=DataInterface() {propertyValues {Curve {keys [
         {interpMode=RCIM_Cubic, leaveTangent=10.6169, tangentMode=RCTM_Break, time=0, value=-0.2}
         {arriveTangent=2.4889, interpMode=RCIM_Cubic, leaveTangent=0.1906, tangentMode=RCTM_Break, time=0.1, value=0.5}
         {arriveTangent=-0.3416, interpMode=RCIM_Cubic, leaveTangent=-1.8967, tangentMode=RCTM_Break, time=0.6, value=0.5}
         {arriveTangent=0.3626, interpMode=RCIM_Cubic, tangentMode=RCTM_Break, time=0.9, value=0}
        ]}}})
    Module.Param0WriteEnabled=true
    Module.Param0WriteEnabled1=true
    Module.Param1WriteEnabled=true
    Module.Param1WriteEnabled1=true
    Module.Param2WriteEnabled=false
    Module.Param2WriteEnabled1=true
    Module.Param3WriteEnabled=false
    Module.Param3WriteEnabled1=false
    Write Parameter Index 1=true
   )
   &ScaleMeshSize(Scale Factor=&MakeVector(
     X=&FloatFromCurve(FloatCurve=DataInterface() {propertyValues {Curve {keys [<at 0.65 1>,<at 1 0.8>]}}})
     Y=&FloatFromCurve(FloatCurve=DataInterface() {propertyValues {Curve {keys [
          {interpMode=RCIM_Cubic, tangentMode=RCTM_User, time=0, value=0.7}
          {interpMode=RCIM_Cubic, tangentMode=RCTM_User, time=0.6, value=1}
          {arriveTangent=-4.4791, interpMode=RCIM_Cubic, leaveTangent=-4.4791, tangentMode=RCTM_User, time=1, value=0}
         ]}}})
     Z=&FloatFromCurve(FloatCurve=DataInterface() {propertyValues {Curve {keys [
          {interpMode=RCIM_Cubic, tangentMode=RCTM_User, time=0, value=0.7}
          {interpMode=RCIM_Cubic, tangentMode=RCTM_User, time=0.6, value=1}
          {arriveTangent=-6.148, interpMode=RCIM_Cubic, leaveTangent=-6.148, tangentMode=RCTM_User, time=1, value=0}
         ]}}})
    ))
  ]
  rendererClasses [&NiagaraMeshRendererProperties]
  renderers [{rendererClass=&NiagaraMeshRendererProperties, rendererIndex=0}]
 }
 Mesh_Calf {
  emitterUpdateScript [
   &EmitterState(Inactive Response=::NewEnumerator1, Life Cycle Mode=::NewEnumerator1, Loop Behavior=::NewEnumerator1, Loop Duration=4, UseLoopDelay=false)
   &SpawnBurst_Instantaneous(Spawn Count=2, Use Spawn Probability=false, UseLoopCountLimit=false)
  ]
  particleSpawnScript [
   &FN_SetSpawnGroup(Number of Spawn Groups=2)
   &InitializeParticle(
    Lifetime Max=2.8
    Lifetime Min=3.3
    Lifetime Mode=::NewEnumerator1
    Mesh Scale=&RandomRangeVector(Minimum=<1.2 0.7 0.7>, Maximum=<1 0.6 0.6>)
    Mesh Scale Mode=::NewEnumerator4
    UsePositionOffset=false
    Write Material Random=true
   )
   &InitialMeshOrientation(
    Orientation Axis=&SelectVectorFromArray(Direct Array Index=$Particles.SpawnGroup, Vector Selection Array=DataInterface() {propertyValues {FloatData [<-1 0 0>, <1 0 0>]}})
    Rotation=&RandomRangeVector(Minimum=<1 0 0>, Maximum=<0 0 0>)
    Rotation Coordinate Space=::NewEnumerator2
    Use Orientation Vector=true
    Use Rotation Vector=true
   )
   &CameraOffset()
  ]
  particleUpdateScript [
   &ParticleState()
   &UpdateMeshOrientation(Rotation Rate=-0.1, Rotation Vector=<1 0 0>)
   &SkeletalMeshLocation(
    Bone / Socket Index=$Particles.SpawnGroup
    Bone Sampling Mode=::NewEnumerator4
    "Interpolate Mesh Sampling (Update)"=true
    MaskSpawnGroup=false
    Sampled Position Offset=&SelectVectorFromArray(
     Direct Array Index=$Particles.SpawnGroup
     Vector Selection Array=DataInterface() {propertyValues {MaxElements=2, FloatData [<-18 0 0>, <18 0 0>]}}
    )
    Skeletal Mesh=DataInterface() {propertyValues {
      DefaultMesh=null
      ExcludeBoneName=root
      OptionalComponentClass=null
      PreviewMesh=&M_MED_Clash_V_Casual
      SkinningMode=SkinOnTheFly
      SoftSourceActor=null
      SourceMode=Default
      UvSetIndex=0
      WholeMeshLOD=-1
      bExcludeBone=false
      bReadDeformedGeometry=true
      bRequireCurrentFrameData=true
      ComponentTags []
      FilteredBones [calf_l,calf_r]
      FilteredSockets []
      MeshUserParameter {parameter {name=None, typeDefHandle {registeredTypeIndex=1}, varData []}}
      SamplingRegions []
     }}
   )
   &DynamicMaterialParameters(
    Index 0 Param 1=&FloatFromCurve(FloatCurve=DataInterface() {propertyValues {Curve {keys [<at 0 1>,<at 0.5 0.5>]}}})
    Index 0 Param 2=&FloatFromCurve(FloatCurve=DataInterface() {propertyValues {Curve {keys [
         {interpMode=RCIM_Cubic, leaveTangent=-2.9591, tangentMode=RCTM_Break, time=0, value=1.5}
         {arriveTangent=-0.3114, interpMode=RCIM_Cubic, tangentMode=RCTM_Break, time=1, value=0}
        ]}}})
    Index 1 Param 1=&FloatFromCurve(FloatCurve=DataInterface() {propertyValues {Curve {keys [
         {arriveTangent=-10.7468, interpMode=RCIM_Cubic, leaveTangent=-10.7468, tangentMode=RCTM_User, time=0, value=2}
         {interpMode=RCIM_Cubic, leaveTangent=0.5305, tangentMode=RCTM_Break, time=0.1, value=1}
         {interpMode=RCIM_Cubic, tangentMode=RCTM_Break, time=1, value=0.2}
        ]}}})
    Index 1 Param 2=2
    Index 1 Param 3=&FloatFromCurve(FloatCurve=DataInterface() {propertyValues {Curve {keys [
         {interpMode=RCIM_Cubic, leaveTangent=10.6169, tangentMode=RCTM_Break, time=0, value=-0.2}
         {arriveTangent=2.4889, interpMode=RCIM_Cubic, leaveTangent=0.1906, tangentMode=RCTM_Break, time=0.1, value=0.5}
         {arriveTangent=-0.3416, interpMode=RCIM_Cubic, leaveTangent=-1.8967, tangentMode=RCTM_Break, time=0.6, value=0.5}
         {arriveTangent=0.3626, interpMode=RCIM_Cubic, tangentMode=RCTM_Break, time=0.9, value=0}
        ]}}})
    Module.Param0WriteEnabled=true
    Module.Param0WriteEnabled1=true
    Module.Param1WriteEnabled=true
    Module.Param1WriteEnabled1=true
    Module.Param2WriteEnabled=false
    Module.Param2WriteEnabled1=true
    Module.Param3WriteEnabled=false
    Module.Param3WriteEnabled1=false
    Write Parameter Index 1=true
   )
   &ScaleMeshSize(Scale Factor=&MakeVector(
     X=&FloatFromCurve(FloatCurve=DataInterface() {propertyValues {Curve {keys [<at 0.65 1>,<at 1 0.8>]}}})
     Y=&FloatFromCurve(FloatCurve=DataInterface() {propertyValues {Curve {keys [
          {interpMode=RCIM_Cubic, tangentMode=RCTM_User, time=0, value=0.7}
          {interpMode=RCIM_Cubic, tangentMode=RCTM_User, time=0.6, value=1}
          {arriveTangent=-4.4791, interpMode=RCIM_Cubic, leaveTangent=-4.4791, tangentMode=RCTM_User, time=1, value=0}
         ]}}})
     Z=&FloatFromCurve(FloatCurve=DataInterface() {propertyValues {Curve {keys [
          {interpMode=RCIM_Cubic, tangentMode=RCTM_User, time=0, value=0.7}
          {interpMode=RCIM_Cubic, tangentMode=RCTM_User, time=0.6, value=1}
          {arriveTangent=-6.148, interpMode=RCIM_Cubic, leaveTangent=-6.148, tangentMode=RCTM_User, time=1, value=0}
         ]}}})
    ))
  ]
  rendererClasses [&NiagaraMeshRendererProperties]
  renderers [{rendererClass=&NiagaraMeshRendererProperties, rendererIndex=0}]
 }
 Mesh_Lowerarm {
  emitterUpdateScript [
   &EmitterState(Inactive Response=::NewEnumerator1, Life Cycle Mode=::NewEnumerator1, Loop Behavior=::NewEnumerator1, Loop Duration=4, UseLoopDelay=false)
   &SpawnBurst_Instantaneous(Spawn Count=2, Spawn Time=0.55, Use Spawn Probability=false, UseLoopCountLimit=false)
  ]
  particleSpawnScript [
   &FN_SetSpawnGroup(Number of Spawn Groups=2)
   &InitializeParticle(
    Lifetime Max=2.3
    Lifetime Min=2.8
    Lifetime Mode=::NewEnumerator1
    Mesh Scale=&RandomRangeVector(Minimum=<0.9 0.55 0.55>, Maximum=<0.8 0.6 0.6>)
    Mesh Scale Mode=::NewEnumerator4
    UsePositionOffset=false
    Write Material Random=true
   )
   &InitialMeshOrientation(
    Orientation Axis=&SelectVectorFromArray(Direct Array Index=$Particles.SpawnGroup, Vector Selection Array=DataInterface() {propertyValues {FloatData [<-1 0 0>, <1 0 0>]}})
    Rotation=&RandomRangeVector(Minimum=<1 0 0>, Maximum=<0 0 0>)
    Rotation Coordinate Space=::NewEnumerator2
    Use Orientation Vector=true
    Use Rotation Vector=true
   )
  ]
  particleUpdateScript [
   &ParticleState()
   &UpdateMeshOrientation(Rotation Rate=-0.1, Rotation Vector=<1 0 0>)
   &SkeletalMeshLocation(
    Bone / Socket Index=$Particles.SpawnGroup
    Bone Sampling Mode=::NewEnumerator4
    "Interpolate Mesh Sampling (Update)"=true
    MaskSpawnGroup=false
    Sampled Position Offset=&SelectVectorFromArray(
     Direct Array Index=$Particles.SpawnGroup
     Vector Selection Array=DataInterface() {propertyValues {MaxElements=2, FloatData [<20 0 0>, <-20 0 0>]}}
    )
    Skeletal Mesh=DataInterface() {propertyValues {
      DefaultMesh=null
      ExcludeBoneName=root
      OptionalComponentClass=null
      PreviewMesh=&M_MED_Clash_V_Casual
      SkinningMode=SkinOnTheFly
      SoftSourceActor=null
      SourceMode=Default
      UvSetIndex=0
      WholeMeshLOD=-1
      bExcludeBone=false
      bReadDeformedGeometry=true
      bRequireCurrentFrameData=true
      ComponentTags []
      FilteredBones [lowerarm_l,lowerarm_r]
      FilteredSockets []
      MeshUserParameter {parameter {name=None, typeDefHandle {registeredTypeIndex=1}, varData []}}
      SamplingRegions []
     }}
   )
   &DynamicMaterialParameters(
    Index 0 Param 1=&FloatFromCurve(FloatCurve=DataInterface() {propertyValues {Curve {keys [<at 0 1>,<at 0.5 0.5>]}}})
    Index 0 Param 2=&FloatFromCurve(FloatCurve=DataInterface() {propertyValues {Curve {keys [
         {interpMode=RCIM_Cubic, leaveTangent=-2.9591, tangentMode=RCTM_Break, time=0, value=1.5}
         {arriveTangent=-0.3114, interpMode=RCIM_Cubic, tangentMode=RCTM_Break, time=1, value=0}
        ]}}})
    Index 1 Param 1=&FloatFromCurve(FloatCurve=DataInterface() {propertyValues {Curve {keys [
         {arriveTangent=-10.7468, interpMode=RCIM_Cubic, leaveTangent=-10.7468, tangentMode=RCTM_User, time=0, value=2}
         {interpMode=RCIM_Cubic, leaveTangent=0.5305, tangentMode=RCTM_Break, time=0.1, value=1}
         {interpMode=RCIM_Cubic, tangentMode=RCTM_Break, time=1, value=0.2}
        ]}}})
    Index 1 Param 2=2
    Index 1 Param 3=&FloatFromCurve(FloatCurve=DataInterface() {propertyValues {Curve {keys [
         {interpMode=RCIM_Cubic, leaveTangent=10.6169, tangentMode=RCTM_Break, time=0, value=-0.2}
         {arriveTangent=2.4889, interpMode=RCIM_Cubic, leaveTangent=0.1906, tangentMode=RCTM_Break, time=0.1, value=0.5}
         {arriveTangent=-0.3416, interpMode=RCIM_Cubic, leaveTangent=-1.8967, tangentMode=RCTM_Break, time=0.6, value=0.5}
         {arriveTangent=0.3626, interpMode=RCIM_Cubic, tangentMode=RCTM_Break, time=0.9, value=0}
        ]}}})
    Module.Param0WriteEnabled=true
    Module.Param0WriteEnabled1=true
    Module.Param1WriteEnabled=true
    Module.Param1WriteEnabled1=true
    Module.Param2WriteEnabled=false
    Module.Param2WriteEnabled1=true
    Module.Param3WriteEnabled=false
    Module.Param3WriteEnabled1=false
    Write Parameter Index 1=true
   )
   &ScaleMeshSize(Scale Factor=&MakeVector(
     X=&FloatFromCurve(FloatCurve=DataInterface() {propertyValues {Curve {keys [<at 0.65 1>,<at 1 0.8>]}}})
     Y=&FloatFromCurve(FloatCurve=DataInterface() {propertyValues {Curve {keys [
          {interpMode=RCIM_Cubic, tangentMode=RCTM_User, time=0, value=0.7}
          {interpMode=RCIM_Cubic, tangentMode=RCTM_User, time=0.6, value=1}
          {arriveTangent=-4.4791, interpMode=RCIM_Cubic, leaveTangent=-4.4791, tangentMode=RCTM_User, time=1, value=0}
         ]}}})
     Z=&FloatFromCurve(FloatCurve=DataInterface() {propertyValues {Curve {keys [
          {interpMode=RCIM_Cubic, tangentMode=RCTM_User, time=0, value=0.7}
          {interpMode=RCIM_Cubic, tangentMode=RCTM_User, time=0.6, value=1}
          {arriveTangent=-6.148, interpMode=RCIM_Cubic, leaveTangent=-6.148, tangentMode=RCTM_User, time=1, value=0}
         ]}}})
    ))
  ]
  rendererClasses [&NiagaraMeshRendererProperties]
  renderers [{rendererClass=&NiagaraMeshRendererProperties, rendererIndex=0}]
 }
 Mesh_Thigh {
  emitterUpdateScript [
   &EmitterState(Inactive Response=::NewEnumerator1, Life Cycle Mode=::NewEnumerator1, Loop Behavior=::NewEnumerator1, Loop Duration=4, UseLoopDelay=false)
   &SpawnBurst_Instantaneous(Spawn Count=2, Spawn Time=0.2, Use Spawn Probability=false, UseLoopCountLimit=false)
  ]
  particleSpawnScript [
   &FN_SetSpawnGroup(Number of Spawn Groups=2)
   &InitializeParticle(
    Lifetime Max=2.5
    Lifetime Min=3
    Lifetime Mode=::NewEnumerator1
    Mesh Scale=&RandomRangeVector(Minimum=<1.4 0.85 0.85>, Maximum=<1.5 0.9 0.9>)
    Mesh Scale Mode=::NewEnumerator4
    UsePositionOffset=false
    Write Material Random=true
   )
   &InitialMeshOrientation(
    Orientation Axis=&SelectVectorFromArray(Direct Array Index=$Particles.SpawnGroup, Vector Selection Array=DataInterface() {propertyValues {FloatData [<-1 0 0>, <1 0 0>]}})
    Rotation=&RandomRangeVector(Minimum=<1 0 0>, Maximum=<0 0 0>)
    Rotation Coordinate Space=::NewEnumerator2
    Use Orientation Vector=true
    Use Rotation Vector=true
   )
   &CameraOffset(Camera Offset Amount=4)
  ]
  particleUpdateScript [
   &ParticleState()
   &UpdateMeshOrientation(Rotation Rate=-0.1, Rotation Vector=<1 0 0>)
   &SkeletalMeshLocation(
    Bone / Socket Index=$Particles.SpawnGroup
    Bone Sampling Mode=::NewEnumerator4
    "Interpolate Mesh Sampling (Update)"=true
    MaskSpawnGroup=false
    Sampled Position Offset=&SelectVectorFromArray(
     Direct Array Index=$Particles.SpawnGroup
     Vector Selection Array=DataInterface() {propertyValues {MaxElements=2, FloatData [<-20 0 0>, <20 0 0>]}}
    )
    Skeletal Mesh=DataInterface() {propertyValues {
      DefaultMesh=null
      ExcludeBoneName=root
      OptionalComponentClass=null
      PreviewMesh=&M_MED_Clash_V_Casual
      SkinningMode=SkinOnTheFly
      SoftSourceActor=null
      SourceMode=Default
      UvSetIndex=0
      WholeMeshLOD=-1
      bExcludeBone=false
      bReadDeformedGeometry=true
      bRequireCurrentFrameData=true
      ComponentTags []
      FilteredBones [thigh_l,thigh_r]
      FilteredSockets []
      MeshUserParameter {parameter {name=None, typeDefHandle {registeredTypeIndex=1}, varData []}}
      SamplingRegions []
     }}
   )
   &DynamicMaterialParameters(
    Index 0 Param 1=&FloatFromCurve(FloatCurve=DataInterface() {propertyValues {Curve {keys [<at 0 1>,<at 0.5 0.5>]}}})
    Index 0 Param 2=&FloatFromCurve(FloatCurve=DataInterface() {propertyValues {Curve {keys [
         {interpMode=RCIM_Cubic, leaveTangent=-2.9591, tangentMode=RCTM_Break, time=0, value=1.5}
         {arriveTangent=-0.3114, interpMode=RCIM_Cubic, tangentMode=RCTM_Break, time=1, value=0}
        ]}}})
    Index 1 Param 1=&FloatFromCurve(FloatCurve=DataInterface() {propertyValues {Curve {keys [
         {arriveTangent=-10.7468, interpMode=RCIM_Cubic, leaveTangent=-10.7468, tangentMode=RCTM_User, time=0, value=2}
         {interpMode=RCIM_Cubic, leaveTangent=0.5305, tangentMode=RCTM_Break, time=0.1, value=1}
         {interpMode=RCIM_Cubic, tangentMode=RCTM_Break, time=1, value=0.2}
        ]}}})
    Index 1 Param 2=2
    Index 1 Param 3=&FloatFromCurve(FloatCurve=DataInterface() {propertyValues {Curve {keys [
         {interpMode=RCIM_Cubic, leaveTangent=10.6169, tangentMode=RCTM_Break, time=0, value=-0.2}
         {arriveTangent=2.4889, interpMode=RCIM_Cubic, leaveTangent=0.1906, tangentMode=RCTM_Break, time=0.1, value=0.5}
         {arriveTangent=-0.3416, interpMode=RCIM_Cubic, leaveTangent=-1.8967, tangentMode=RCTM_Break, time=0.6, value=0.5}
         {arriveTangent=0.3626, interpMode=RCIM_Cubic, tangentMode=RCTM_Break, time=0.9, value=0}
        ]}}})
    Module.Param0WriteEnabled=true
    Module.Param0WriteEnabled1=true
    Module.Param1WriteEnabled=true
    Module.Param1WriteEnabled1=true
    Module.Param2WriteEnabled=false
    Module.Param2WriteEnabled1=true
    Module.Param3WriteEnabled=false
    Module.Param3WriteEnabled1=false
    Write Parameter Index 1=true
   )
   &ScaleMeshSize(Scale Factor=&MakeVector(
     X=&FloatFromCurve(FloatCurve=DataInterface() {propertyValues {Curve {keys [<at 0.65 1>,<at 1 0.8>]}}})
     Y=&FloatFromCurve(FloatCurve=DataInterface() {propertyValues {Curve {keys [
          {interpMode=RCIM_Cubic, tangentMode=RCTM_User, time=0, value=0.7}
          {interpMode=RCIM_Cubic, tangentMode=RCTM_User, time=0.6, value=1}
          {arriveTangent=-4.4791, interpMode=RCIM_Cubic, leaveTangent=-4.4791, tangentMode=RCTM_User, time=1, value=0}
         ]}}})
     Z=&FloatFromCurve(FloatCurve=DataInterface() {propertyValues {Curve {keys [
          {interpMode=RCIM_Cubic, tangentMode=RCTM_User, time=0, value=0.7}
          {interpMode=RCIM_Cubic, tangentMode=RCTM_User, time=0.6, value=1}
          {arriveTangent=-6.148, interpMode=RCIM_Cubic, leaveTangent=-6.148, tangentMode=RCTM_User, time=1, value=0}
         ]}}})
    ))
  ]
  rendererClasses [&NiagaraMeshRendererProperties]
  renderers [{rendererClass=&NiagaraMeshRendererProperties, rendererIndex=0}]
 }
 Mesh_Upperarm {
  emitterUpdateScript [
   &EmitterState(Inactive Response=::NewEnumerator1, Life Cycle Mode=::NewEnumerator1, Loop Behavior=::NewEnumerator1, Loop Duration=4, UseLoopDelay=false)
   &SpawnBurst_Instantaneous(Spawn Count=2, Spawn Time=0.4, Use Spawn Probability=false, UseLoopCountLimit=false)
  ]
  particleSpawnScript [
   &FN_SetSpawnGroup(Number of Spawn Groups=2)
   &InitializeParticle(
    Lifetime Max=2.8
    Lifetime Min=2.5
    Lifetime Mode=::NewEnumerator1
    Mesh Scale=&RandomRangeVector(Minimum=<0.9 0.9 0.9>, Maximum=<1.2 0.95 0.95>)
    Mesh Scale Mode=::NewEnumerator4
    UsePositionOffset=false
    Write Material Random=true
   )
   &InitialMeshOrientation(
    Orientation Axis=&SelectVectorFromArray(Direct Array Index=$Particles.SpawnGroup, Vector Selection Array=DataInterface() {propertyValues {FloatData [<-1 0 0>, <1 0 0>]}})
    Rotation=&RandomRangeVector(Minimum=<1 0 0>, Maximum=<0 0 0>)
    Rotation Coordinate Space=::NewEnumerator2
    Use Orientation Vector=true
    Use Rotation Vector=true
   )
  ]
  particleUpdateScript [
   &ParticleState()
   &UpdateMeshOrientation(Rotation Rate=-0.1, Rotation Vector=<1 0 0>)
   &SkeletalMeshLocation(
    Bone / Socket Index=$Particles.SpawnGroup
    Bone Sampling Mode=::NewEnumerator4
    "Interpolate Mesh Sampling (Update)"=true
    MaskSpawnGroup=false
    Sampled Position Offset=&SelectVectorFromArray(
     Direct Array Index=$Particles.SpawnGroup
     Vector Selection Array=DataInterface() {propertyValues {MaxElements=2, FloatData [<10 0 0>, <-10 0 0>]}}
    )
    Skeletal Mesh=DataInterface() {propertyValues {
      DefaultMesh=null
      ExcludeBoneName=root
      OptionalComponentClass=null
      PreviewMesh=&M_MED_Clash_V_Casual
      SkinningMode=SkinOnTheFly
      SoftSourceActor=null
      SourceMode=Default
      UvSetIndex=0
      WholeMeshLOD=-1
      bExcludeBone=false
      bReadDeformedGeometry=true
      bRequireCurrentFrameData=true
      ComponentTags []
      FilteredBones [upperarm_l,upperarm_r]
      FilteredSockets []
      MeshUserParameter {parameter {name=None, typeDefHandle {registeredTypeIndex=1}, varData []}}
      SamplingRegions []
     }}
   )
   &DynamicMaterialParameters(
    Index 0 Param 1=&FloatFromCurve(FloatCurve=DataInterface() {propertyValues {Curve {keys [<at 0 1>,<at 0.5 0.5>]}}})
    Index 0 Param 2=&FloatFromCurve(FloatCurve=DataInterface() {propertyValues {Curve {keys [
         {interpMode=RCIM_Cubic, leaveTangent=-2.9591, tangentMode=RCTM_Break, time=0, value=1.5}
         {arriveTangent=-0.3114, interpMode=RCIM_Cubic, tangentMode=RCTM_Break, time=1, value=0}
        ]}}})
    Index 1 Param 1=&FloatFromCurve(FloatCurve=DataInterface() {propertyValues {Curve {keys [
         {arriveTangent=-10.7468, interpMode=RCIM_Cubic, leaveTangent=-10.7468, tangentMode=RCTM_User, time=0, value=2}
         {interpMode=RCIM_Cubic, leaveTangent=0.5305, tangentMode=RCTM_Break, time=0.1, value=1}
         {interpMode=RCIM_Cubic, tangentMode=RCTM_Break, time=1, value=0.2}
        ]}}})
    Index 1 Param 2=2
    Index 1 Param 3=&FloatFromCurve(FloatCurve=DataInterface() {propertyValues {Curve {keys [
         {interpMode=RCIM_Cubic, leaveTangent=10.6169, tangentMode=RCTM_Break, time=0, value=-0.2}
         {arriveTangent=2.4889, interpMode=RCIM_Cubic, leaveTangent=0.1906, tangentMode=RCTM_Break, time=0.1, value=0.5}
         {arriveTangent=-0.3416, interpMode=RCIM_Cubic, leaveTangent=-1.8967, tangentMode=RCTM_Break, time=0.6, value=0.5}
         {arriveTangent=0.3626, interpMode=RCIM_Cubic, tangentMode=RCTM_Break, time=0.9, value=0}
        ]}}})
    Module.Param0WriteEnabled=true
    Module.Param0WriteEnabled1=true
    Module.Param1WriteEnabled=true
    Module.Param1WriteEnabled1=true
    Module.Param2WriteEnabled=false
    Module.Param2WriteEnabled1=true
    Module.Param3WriteEnabled=false
    Module.Param3WriteEnabled1=false
    Write Parameter Index 1=true
   )
   &ScaleMeshSize(Scale Factor=&MakeVector(
     X=&FloatFromCurve(FloatCurve=DataInterface() {propertyValues {Curve {keys [<at 0.65 1>,<at 1 0.8>]}}})
     Y=&FloatFromCurve(FloatCurve=DataInterface() {propertyValues {Curve {keys [
          {interpMode=RCIM_Cubic, tangentMode=RCTM_User, time=0, value=0.7}
          {interpMode=RCIM_Cubic, tangentMode=RCTM_User, time=0.6, value=1}
          {arriveTangent=-4.4791, interpMode=RCIM_Cubic, leaveTangent=-4.4791, tangentMode=RCTM_User, time=1, value=0}
         ]}}})
     Z=&FloatFromCurve(FloatCurve=DataInterface() {propertyValues {Curve {keys [
          {interpMode=RCIM_Cubic, tangentMode=RCTM_User, time=0, value=0.7}
          {interpMode=RCIM_Cubic, tangentMode=RCTM_User, time=0.6, value=1}
          {arriveTangent=-6.148, interpMode=RCIM_Cubic, leaveTangent=-6.148, tangentMode=RCTM_User, time=1, value=0}
         ]}}})
    ))
  ]
  rendererClasses [&NiagaraMeshRendererProperties]
  renderers [{rendererClass=&NiagaraMeshRendererProperties, rendererIndex=0}]
 }
}
summary {
 systemName=NS_Emote_Clash
 emitters [
  {emitterName=BoneFlow_Back, rendererClasses [&NiagaraRibbonRendererProperties]}
  {emitterName=BoneFlow_Leg_L, rendererClasses [&NiagaraMeshRendererProperties]}
  {emitterName=BoneFlow_Leg_R, rendererClasses [&NiagaraMeshRendererProperties]}
  {emitterName=BoneFlow_Body, rendererClasses [&NiagaraMeshRendererProperties]}
  {emitterName=BoneFlow_Arm_L, rendererClasses [&NiagaraMeshRendererProperties]}
  {emitterName=BoneFlow_Arm_R, rendererClasses [&NiagaraMeshRendererProperties]}
  {emitterName=Mesh_Calf, rendererClasses [&NiagaraMeshRendererProperties]}
  {emitterName=Mesh_Thigh, rendererClasses [&NiagaraMeshRendererProperties]}
  {emitterName=Mesh_Upperarm, rendererClasses [&NiagaraMeshRendererProperties]}
  {emitterName=Mesh_Lowerarm, rendererClasses [&NiagaraMeshRendererProperties]}
  {emitterName=Mesh_Body, rendererClasses [&NiagaraMeshRendererProperties]}
  {emitterName=BoneFlow_LHand, rendererClasses [&NiagaraRibbonRendererProperties]}
  {emitterName=BoneFlow_RHand, rendererClasses [&NiagaraRibbonRendererProperties]}
 ]
 userVariables []
}
