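# Niagara mesh-particle effect for the Clash emote, apparently played on removal, that covers the character's limbs and body.
# Five BoneFlow emitters briefly spawn grouped meshes, placing them along filtered arm, leg and head-to-pelvis bone chains via FN_BoneFlow.
# Spline tangents drive a randomly rolled ring rotation, while curves over lifetime and group position shape scale and dynamic material parameters.
# Mesh_Calf bursts two slowly rotating meshes that follow the calves, then shrink.
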NiagaraSystem /BRCosmetics/Athena/Effects/Particles/Emotes/NS_Emote_Clash/NS_Emote_Clash_Removal
dependencies {
 usedDataInterfaces [&NiagaraDataInterfaceCurve, &NiagaraDataInterfaceSkeletalMesh, &NiagaraDataInterfaceArrayFloat3]
 usedDynamicInputs []
 usedModules [&SystemState]
}
emitters {
 BoneFlow_Arm_L {
  emitterUpdateScript [
   &EmitterState(Life Cycle Mode=::NewEnumerator1, Loop Behavior=::NewEnumerator1, Loop Duration=0.4, UseLoopDelay=true)
   &SpawnRate(SpawnRate=42, Use Spawn Probability=false)
  ]
  particleSpawnScript [
   &FN_SetSpawnGroup(Inclusivity=true, Normalize Spawn Group=true, Particles Per Group=16, Pattern=::NewEnumerator0, Set Ribbon ID and Ribbon Link Order=true) {module=FN_SetSpawnGroup001}
   &InitializeParticle(
    Color=<rgba 0.9249 1 0.9295 1>
    Lifetime=&FloatFromCurve(
     FloatCurve=DataInterface() {propertyValues {Curve {keys [<at 0 0.3>,<at 0.7 0.4>,<at 0.9 0.4>]}}}
     CurveIndex=&RandomRangeFloat(Minimum=0, Maximum=1)
    )
    Mesh Scale Max=<0.8 0.8 2.5>
    Mesh Scale Min=<0.6 0.6 2>
    Mesh Scale Mode=::NewEnumerator2
    UsePositionOffset=false
    Write Material Random=true
   )
   &InitialMeshOrientation(Mesh Orientation Mode=::NewEnumerator2, Use Rotation Vector=true)
   &CameraOffset(Camera Offset Amount=10)
  ]
  particleUpdateScript [
   &ParticleState()
   /Game/Effects/Niagara/Modules/BoneFlow/FN_BoneFlow_v4/FN_BoneFlow(
    Angle:NiagaraFloat=&Multiply_Float(A=$Particles.Normalized Spawn Group, B=2)
    BoneID:NiagaraInt32=&ModuloInt(A=$Particles.SpawnGroup, B=1)
    Height:NiagaraFloat=&Multiply_Float(A=$Particles.Normalized Spawn Group, B=1)
    Interpolate Mesh Sampling=false
    Radius:NiagaraFloat=&Multiply_Float(
     A=&FloatFromCurve(
      FloatCurve=DataInterface() {propertyValues {Curve {keys [
          {arriveTangent=64.149, interpMode=RCIM_Cubic, leaveTangent=64.149, tangentMode=RCTM_User, time=0, value=0}
          <at 0.5 10 RCIM_Cubic>
          {arriveTangent=-53.2228, interpMode=RCIM_Cubic, leaveTangent=-53.2228, tangentMode=RCTM_User, time=1, value=0}
         ]}}}
      CurveIndex=$Particles.Normalized Spawn Group
     )
     B=0.5
    )
    SideVector2:Vector3f=<0 0 0>
    SideVector3:Vector3f=<1 1 1>
    SkeletalMesh=DataInterface() {propertyValues {
      DefaultMesh=null
      ExcludeBoneName=root
      OptionalComponentClass=null
      PreviewMesh=/BRCosmetics/Characters/Player/Male/Medium/Bodies/M_MED_Clash_V_Casual/Meshes/M_MED_Clash_V_Casual
      SkinningMode=SkinOnTheFly
      SoftSourceActor=null
      SourceMode=Default
      UvSetIndex=0
      WholeMeshLOD=-1
      bExcludeBone=false
      bReadDeformedGeometry=true
      bRequireCurrentFrameData=true
      ComponentTags []
      FilteredBones [hand_l,lowerarm_l,upperarm_l]
      FilteredSockets []
      MeshUserParameter {parameter {name=None, typeDefHandle {registeredTypeIndex=1}, varData []}}
      SamplingRegions []
     }}
    Stabalize Twisting=true
    Write Position Attribute=true
   ) {module=FN_BoneFlow001}
   /SetVariables(Particles.BoneFlowTangent:Vector3f=$Output.FN_BoneFlow001.SplineTangent) {isSetParameters=true, module=SetVariables_#1}
   /SetVariables(Particles.RingRotation:Quat4f=&MultiplyQuaternion(
     Quaternion A=&MakeQuatFromAxisAngle(Axis=$Particles.BoneFlowTangent, AngleInDegrees=&RandomRangeFloat(Minimum=0, Maximum=360))
     Quaternion B=&FindQuatBetweenTwoVectors(From Vector=<0.4 0.4 1>, To Vector=$Particles.BoneFlowTangent)
    )) {isSetParameters=true, module=SetVariables_#2}
   /SetVariables(Particles.ScaleMeshShapeCurve=DataInterface() {propertyValues {Curve {keys [<at 0 1>,<at 1 1>]}}}) {isSetParameters=true, module=SetVariables_#3}
   /SetVariables(Particles.ScaleMeshXYCurve=DataInterface() {propertyValues {Curve {keys [
        {interpMode=RCIM_Cubic, tangentMode=RCTM_User, time=0, value=0}
        {interpMode=RCIM_Cubic, tangentMode=RCTM_User, time=0.35, value=0.7}
        {arriveTangent=-0.2106, interpMode=RCIM_Cubic, leaveTangent=-0.2106, tangentMode=RCTM_User, time=1, value=1.2}
       ]}}}) {isSetParameters=true, module=SetVariables_#4}
   &ScaleMeshSize(Scale Factor=&Multiply_Vector(
     A=&MakeVector(
      X=&FloatFromCurve(FloatCurve=$Particles.ScaleMeshXYCurve)
      Y=&FloatFromCurve(FloatCurve=$Particles.ScaleMeshXYCurve)
      Z=&FloatFromCurve(FloatCurve=DataInterface() {propertyValues {Curve {keys [<at 0 1>, {interpMode=RCIM_Cubic, tangentMode=RCTM_User, time=0.6, value=1}, {interpMode=RCIM_Cubic, tangentMode=RCTM_User, time=1, value=0}]}}})
     )
     B=&MakeVector(
      X=&FloatFromCurve(FloatCurve=$Particles.ScaleMeshShapeCurve, CurveIndex=$Particles.Normalized Spawn Group)
      Y=&FloatFromCurve(FloatCurve=$Particles.ScaleMeshShapeCurve, CurveIndex=$Particles.Normalized Spawn Group)
      Z=1
     )
    ))
   &DynamicMaterialParameters(
    Index 0 Param 1=0.5
    Index 0 Param 2=&FloatFromCurve(FloatCurve=DataInterface() {propertyValues {Curve {keys [
         {interpMode=RCIM_Cubic, tangentMode=RCTM_User, time=0, value=0.5}
         {arriveTangent=2.7221, interpMode=RCIM_Cubic, leaveTangent=2.7221, tangentMode=RCTM_User, time=0.5, value=1.5}
         {arriveTangent=0.4617, interpMode=RCIM_Cubic, leaveTangent=0.4617, tangentMode=RCTM_User, time=0.95, value=2}
         {interpMode=RCIM_Cubic, tangentMode=RCTM_User, time=1, value=2.5}
        ]}}})
    Index 1 Param 1=&FloatFromCurve(FloatCurve=DataInterface() {propertyValues {Curve {keys [<at 0 2>,<at 0.4 0.5>]}}})
    Index 1 Param 2=2
    Index 1 Param 3=&FloatFromCurve(FloatCurve=DataInterface() {propertyValues {Curve {keys [<at 0 8>,<at 0.3 1>]}}})
    Module.Param0WriteEnabled=true
    Module.Param0WriteEnabled1=true
    Module.Param1WriteEnabled=true
    Module.Param1WriteEnabled1=true
    Module.Param2WriteEnabled=true
    Module.Param2WriteEnabled1=true
    Module.Param3WriteEnabled=true
    Module.Param3WriteEnabled1=true
    Write Parameter Index 1=true
   )
  ]
  rendererClasses [&NiagaraMeshRendererProperties]
  renderers [{rendererClass=&NiagaraMeshRendererProperties, rendererIndex=0}]
 }
 BoneFlow_Arm_R {
  emitterUpdateScript [
   &EmitterState(Life Cycle Mode=::NewEnumerator1, Loop Behavior=::NewEnumerator1, Loop Duration=0.4, UseLoopDelay=true)
   &SpawnRate(SpawnRate=42, Use Spawn Probability=false)
  ]
  particleSpawnScript [
   &FN_SetSpawnGroup(Inclusivity=true, Normalize Spawn Group=true, Particles Per Group=16, Pattern=::NewEnumerator0, Set Ribbon ID and Ribbon Link Order=true) {module=FN_SetSpawnGroup001}
   &InitializeParticle(
    Color=<rgba 0.9249 1 0.9295 1>
    Lifetime=&FloatFromCurve(
     FloatCurve=DataInterface() {propertyValues {Curve {keys [<at 0 0.3>,<at 0.7 0.4>,<at 0.9 0.4>]}}}
     CurveIndex=&RandomRangeFloat(Minimum=0, Maximum=1)
    )
    Mesh Scale Max=<0.8 0.8 2.5>
    Mesh Scale Min=<0.6 0.6 2>
    Mesh Scale Mode=::NewEnumerator2
    UsePositionOffset=false
    Write Material Random=true
   )
   &InitialMeshOrientation(Mesh Orientation Mode=::NewEnumerator2, Use Rotation Vector=true)
   &CameraOffset(Camera Offset Amount=10)
  ]
  particleUpdateScript [
   &ParticleState()
   &FN_BoneFlow(
    Angle:NiagaraFloat=&Multiply_Float(A=$Particles.Normalized Spawn Group, B=2)
    BoneID:NiagaraInt32=&ModuloInt(A=$Particles.SpawnGroup, B=1)
    Height:NiagaraFloat=&Multiply_Float(A=$Particles.Normalized Spawn Group, B=1)
    Interpolate Mesh Sampling=false
    Radius:NiagaraFloat=&Multiply_Float(
     A=&FloatFromCurve(
      FloatCurve=DataInterface() {propertyValues {Curve {keys [
          {arriveTangent=64.149, interpMode=RCIM_Cubic, leaveTangent=64.149, tangentMode=RCTM_User, time=0, value=0}
          <at 0.5 10 RCIM_Cubic>
          {arriveTangent=-53.2228, interpMode=RCIM_Cubic, leaveTangent=-53.2228, tangentMode=RCTM_User, time=1, value=0}
         ]}}}
      CurveIndex=$Particles.Normalized Spawn Group
     )
     B=0.5
    )
    SideVector2:Vector3f=<0 0 0>
    SideVector3:Vector3f=<1 1 1>
    SkeletalMesh=DataInterface() {propertyValues {
      DefaultMesh=null
      ExcludeBoneName=root
      OptionalComponentClass=null
      PreviewMesh=&M_MED_Clash_V_Casual
      SkinningMode=SkinOnTheFly
      SoftSourceActor=null
      SourceMode=Default
      UvSetIndex=0
      WholeMeshLOD=-1
      bExcludeBone=false
      bReadDeformedGeometry=true
      bRequireCurrentFrameData=true
      ComponentTags []
      FilteredBones [hand_r,lowerarm_r,upperarm_r]
      FilteredSockets []
      MeshUserParameter {parameter {name=None, typeDefHandle {registeredTypeIndex=1}, varData []}}
      SamplingRegions []
     }}
    Stabalize Twisting=true
    Write Position Attribute=true
   ) {module=FN_BoneFlow001}
   /SetVariables(Particles.BoneFlowTangent:Vector3f=$Output.FN_BoneFlow001.SplineTangent) {isSetParameters=true, module=SetVariables_#1}
   /SetVariables(Particles.RingRotation:Quat4f=&MultiplyQuaternion(
     Quaternion A=&MakeQuatFromAxisAngle(Axis=$Particles.BoneFlowTangent, AngleInDegrees=&RandomRangeFloat(Minimum=0, Maximum=360))
     Quaternion B=&FindQuatBetweenTwoVectors(From Vector=<0.4 0.4 1>, To Vector=$Particles.BoneFlowTangent)
    )) {isSetParameters=true, module=SetVariables_#2}
   /SetVariables(Particles.ScaleMeshShapeCurve=DataInterface() {propertyValues {Curve {keys [<at 0 1>,<at 1 1>]}}}) {isSetParameters=true, module=SetVariables_#3}
   /SetVariables(Particles.ScaleMeshXYCurve=DataInterface() {propertyValues {Curve {keys [
        {interpMode=RCIM_Cubic, tangentMode=RCTM_User, time=0, value=0}
        {interpMode=RCIM_Cubic, tangentMode=RCTM_User, time=0.35, value=0.7}
        {arriveTangent=-0.2106, interpMode=RCIM_Cubic, leaveTangent=-0.2106, tangentMode=RCTM_User, time=1, value=1.2}
       ]}}}) {isSetParameters=true, module=SetVariables_#4}
   &ScaleMeshSize(Scale Factor=&Multiply_Vector(
     A=&MakeVector(
      X=&FloatFromCurve(FloatCurve=$Particles.ScaleMeshXYCurve)
      Y=&FloatFromCurve(FloatCurve=$Particles.ScaleMeshXYCurve)
      Z=&FloatFromCurve(FloatCurve=DataInterface() {propertyValues {Curve {keys [<at 0 1>, {interpMode=RCIM_Cubic, tangentMode=RCTM_User, time=0.6, value=1}, {interpMode=RCIM_Cubic, tangentMode=RCTM_User, time=1, value=0}]}}})
     )
     B=&MakeVector(
      X=&FloatFromCurve(FloatCurve=$Particles.ScaleMeshShapeCurve, CurveIndex=$Particles.Normalized Spawn Group)
      Y=&FloatFromCurve(FloatCurve=$Particles.ScaleMeshShapeCurve, CurveIndex=$Particles.Normalized Spawn Group)
      Z=1
     )
    ))
   &DynamicMaterialParameters(
    Index 0 Param 1=0.5
    Index 0 Param 2=&FloatFromCurve(FloatCurve=DataInterface() {propertyValues {Curve {keys [
         {interpMode=RCIM_Cubic, tangentMode=RCTM_User, time=0, value=0.5}
         {arriveTangent=2.7221, interpMode=RCIM_Cubic, leaveTangent=2.7221, tangentMode=RCTM_User, time=0.5, value=1.5}
         {arriveTangent=0.4617, interpMode=RCIM_Cubic, leaveTangent=0.4617, tangentMode=RCTM_User, time=0.95, value=2}
         {interpMode=RCIM_Cubic, tangentMode=RCTM_User, time=1, value=2.5}
        ]}}})
    Index 1 Param 1=&FloatFromCurve(FloatCurve=DataInterface() {propertyValues {Curve {keys [<at 0 2>,<at 0.4 0.5>]}}})
    Index 1 Param 2=2
    Index 1 Param 3=&FloatFromCurve(FloatCurve=DataInterface() {propertyValues {Curve {keys [<at 0 8>,<at 0.3 1>]}}})
    Module.Param0WriteEnabled=true
    Module.Param0WriteEnabled1=true
    Module.Param1WriteEnabled=true
    Module.Param1WriteEnabled1=true
    Module.Param2WriteEnabled=true
    Module.Param2WriteEnabled1=true
    Module.Param3WriteEnabled=true
    Module.Param3WriteEnabled1=true
    Write Parameter Index 1=true
   )
  ]
  rendererClasses [&NiagaraMeshRendererProperties]
  renderers [{rendererClass=&NiagaraMeshRendererProperties, rendererIndex=0}]
 }
 BoneFlow_Body {
  emitterUpdateScript [
   &EmitterState(Life Cycle Mode=::NewEnumerator1, Loop Behavior=::NewEnumerator1, Loop Delay=0.1, Loop Duration=0.65, UseLoopDelay=true)
   &SpawnRate(SpawnRate=22, Use Spawn Probability=false)
  ]
  particleSpawnScript [
   &FN_SetSpawnGroup(Inclusivity=true, Normalize Spawn Group=true, Particles Per Group=15, Pattern=::NewEnumerator0, Set Ribbon ID and Ribbon Link Order=true) {module=FN_SetSpawnGroup001}
   &InitializeParticle(
    Color=<rgba 0.9249 1 0.9295 1>
    Lifetime=&FloatFromCurve(
     FloatCurve=DataInterface() {propertyValues {Curve {keys [<at 0 0.4>,<at 0.8 0.5>,<at 1 0.9>]}}}
     CurveIndex=&RandomRangeFloat(Minimum=0, Maximum=1)
    )
    Mesh Scale Max=<2.7 3 8>
    Mesh Scale Min=<2.3 2.5 9>
    Mesh Scale Mode=::NewEnumerator2
    UsePositionOffset=false
    Write Material Random=true
   )
   &InitialMeshOrientation(Mesh Orientation Mode=::NewEnumerator2, Use Rotation Vector=true)
   &CameraOffset(Camera Offset Amount=10)
  ]
  particleUpdateScript [
   &ParticleState()
   &FN_BoneFlow(
    Angle:NiagaraFloat=&Multiply_Float(A=$Particles.Normalized Spawn Group, B=2)
    BoneID:NiagaraInt32=&ModuloInt(A=$Particles.SpawnGroup, B=1)
    Height:NiagaraFloat=&Multiply_Float(A=$Particles.Normalized Spawn Group, B=1.1)
    Interpolate Mesh Sampling=false
    Radius:NiagaraFloat=&Multiply_Float(
     A=&FloatFromCurve(
      FloatCurve=DataInterface() {propertyValues {Curve {keys [
          {arriveTangent=64.149, interpMode=RCIM_Cubic, leaveTangent=64.149, tangentMode=RCTM_User, time=0, value=0}
          <at 0.5 10 RCIM_Cubic>
          {arriveTangent=0.3152, interpMode=RCIM_Cubic, leaveTangent=0.3152, tangentMode=RCTM_User, time=0.8, value=0}
         ]}}}
      CurveIndex=$Particles.Normalized Spawn Group
     )
     B=1
    )
    SideVector2:Vector3f=<0 0 0>
    SideVector3:Vector3f=<1 1 1>
    SkeletalMesh=DataInterface() {propertyValues {
      DefaultMesh=null
      ExcludeBoneName=root
      OptionalComponentClass=null
      PreviewMesh=&M_MED_Clash_V_Casual
      SkinningMode=SkinOnTheFly
      SoftSourceActor=null
      SourceMode=Default
      UvSetIndex=0
      WholeMeshLOD=-1
      bExcludeBone=false
      bReadDeformedGeometry=true
      bRequireCurrentFrameData=true
      ComponentTags []
      FilteredBones [head,spine_03,pelvis]
      FilteredSockets []
      MeshUserParameter {parameter {name=None, typeDefHandle {registeredTypeIndex=1}, varData []}}
      SamplingRegions []
     }}
    Stabalize Twisting=true
    Write Position Attribute=true
   ) {module=FN_BoneFlow001}
   /SetVariables(Particles.BoneFlowTangent:Vector3f=$Output.FN_BoneFlow001.SplineTangent) {isSetParameters=true, module=SetVariables_#1}
   /SetVariables(Particles.RingRotation:Quat4f=&MultiplyQuaternion(
     Quaternion A=&MakeQuatFromAxisAngle(Axis=$Particles.BoneFlowTangent, AngleInDegrees=&RandomRangeFloat(Minimum=0, Maximum=360))
     Quaternion B=&FindQuatBetweenTwoVectors(From Vector=<0.2 0.2 1>, To Vector=$Particles.BoneFlowTangent)
    )) {isSetParameters=true, module=SetVariables_#2}
   /SetVariables(Particles.ScaleMeshShapeCurve=DataInterface() {propertyValues {Curve {keys [<at 0 1.2>,<at 0.7 1.1>,<at 0.77 0.5>]}}}) {isSetParameters=true, module=SetVariables_#3}
   /SetVariables(Particles.ScaleMeshXYCurve=DataInterface() {propertyValues {Curve {keys [
        {interpMode=RCIM_Cubic, tangentMode=RCTM_User, time=0, value=0}
        {interpMode=RCIM_Cubic, tangentMode=RCTM_User, time=0.35, value=0.7}
        {arriveTangent=-0.2106, interpMode=RCIM_Cubic, leaveTangent=-0.2106, tangentMode=RCTM_User, time=1, value=1}
       ]}}}) {isSetParameters=true, module=SetVariables_#4}
   &ScaleMeshSize(Scale Factor=&Multiply_Vector(
     A=&MakeVector(
      X=&FloatFromCurve(FloatCurve=$Particles.ScaleMeshXYCurve)
      Y=&FloatFromCurve(FloatCurve=$Particles.ScaleMeshXYCurve)
      Z=&FloatFromCurve(FloatCurve=DataInterface() {propertyValues {Curve {keys [<at 0 1>, {interpMode=RCIM_Cubic, tangentMode=RCTM_User, time=0.5, value=1}, {interpMode=RCIM_Cubic, tangentMode=RCTM_User, time=1, value=0}]}}})
     )
     B=&MakeVector(
      X=&FloatFromCurve(FloatCurve=$Particles.ScaleMeshShapeCurve, CurveIndex=$Particles.Normalized Spawn Group)
      Y=&FloatFromCurve(FloatCurve=$Particles.ScaleMeshShapeCurve, CurveIndex=$Particles.Normalized Spawn Group)
      Z=&FloatFromCurve(
       FloatCurve=DataInterface() {propertyValues {Curve {keys [
           <at 0 1>
           {interpMode=RCIM_Cubic, tangentMode=RCTM_User, time=0.6, value=0.8}
           {interpMode=RCIM_Cubic, tangentMode=RCTM_User, time=0.7, value=0.2}
          ]}}}
       CurveIndex=$Particles.Normalized Spawn Group
      )
     )
    ))
   &DynamicMaterialParameters(
    Index 0 Param 1=0.5
    Index 0 Param 2=&FloatFromCurve(FloatCurve=DataInterface() {propertyValues {Curve {keys [
         {interpMode=RCIM_Cubic, tangentMode=RCTM_User, time=0, value=0.5}
         {arriveTangent=2.7221, interpMode=RCIM_Cubic, leaveTangent=2.7221, tangentMode=RCTM_User, time=0.5, value=1.5}
         {arriveTangent=0.4617, interpMode=RCIM_Cubic, leaveTangent=0.4617, tangentMode=RCTM_User, time=0.95, value=2}
         {interpMode=RCIM_Cubic, tangentMode=RCTM_User, time=1, value=2.5}
        ]}}})
    Index 1 Param 1=&FloatFromCurve(FloatCurve=DataInterface() {propertyValues {Curve {keys [<at 0 2>,<at 0.4 0.5>]}}})
    Index 1 Param 2=2
    Index 1 Param 3=&FloatFromCurve(FloatCurve=DataInterface() {propertyValues {Curve {keys [<at 0 0>,<at 0.1 12>,<at 0.3 1>]}}})
    Module.Param0WriteEnabled=true
    Module.Param0WriteEnabled1=true
    Module.Param1WriteEnabled=true
    Module.Param1WriteEnabled1=true
    Module.Param2WriteEnabled=true
    Module.Param2WriteEnabled1=true
    Module.Param3WriteEnabled=true
    Module.Param3WriteEnabled1=true
    Write Parameter Index 1=true
   )
  ]
  rendererClasses [&NiagaraMeshRendererProperties]
  renderers [{rendererClass=&NiagaraMeshRendererProperties, rendererIndex=0}]
 }
 BoneFlow_Leg_L {
  emitterUpdateScript [
   &EmitterState(Life Cycle Mode=::NewEnumerator1, Loop Behavior=::NewEnumerator1, Loop Duration=0.4, UseLoopDelay=false)
   &SpawnRate(SpawnRate=50, Use Spawn Probability=false)
  ]
  particleSpawnScript [
   &FN_SetSpawnGroup(Inclusivity=true, Normalize Spawn Group=true, Particles Per Group=20, Pattern=::NewEnumerator0, Set Ribbon ID and Ribbon Link Order=true) {module=FN_SetSpawnGroup001}
   &InitializeParticle(
    Color=<rgba 0.9249 1 0.9295 1>
    Lifetime=&FloatFromCurve(
     FloatCurve=DataInterface() {propertyValues {Curve {keys [<at 0 0.3>,<at 0.7 0.6>,<at 1 1.2>]}}}
     CurveIndex=&RandomRangeFloat(Minimum=0, Maximum=1)
    )
    Mesh Scale Max=<1.4 1.4 3>
    Mesh Scale Min=<0.9 0.9 2>
    Mesh Scale Mode=::NewEnumerator2
    UsePositionOffset=false
    Write Material Random=true
   )
   &InitialMeshOrientation(Mesh Orientation Mode=::NewEnumerator2, Use Rotation Vector=true)
   &CameraOffset(Camera Offset Amount=10)
  ]
  particleUpdateScript [
   &ParticleState()
   &FN_BoneFlow(
    Angle:NiagaraFloat=&Multiply_Float(A=$Particles.Normalized Spawn Group, B=-1)
    BoneID:NiagaraInt32=&ModuloInt(A=$Particles.SpawnGroup, B=1)
    Height:NiagaraFloat=&Multiply_Float(A=$Particles.Normalized Spawn Group, B=1.2)
    Interpolate Mesh Sampling=false
    Radius:NiagaraFloat=&Multiply_Float(
     A=&FloatFromCurve(
      FloatCurve=DataInterface() {propertyValues {Curve {keys [
          {arriveTangent=64.149, interpMode=RCIM_Cubic, leaveTangent=64.149, tangentMode=RCTM_User, time=0, value=0}
          <at 0.5 10 RCIM_Cubic>
          {arriveTangent=-53.2228, interpMode=RCIM_Cubic, leaveTangent=-53.2228, tangentMode=RCTM_User, time=1, value=0}
         ]}}}
      CurveIndex=$Particles.Normalized Spawn Group
     )
     B=1
    )
    SideVector2:Vector3f=<0 0 0>
    SideVector3:Vector3f=<1 1 1>
    SkeletalMesh=DataInterface() {propertyValues {
      DefaultMesh=null
      ExcludeBoneName=root
      OptionalComponentClass=null
      PreviewMesh=/Game/Characters/Player/Male/Medium/Base/SK_M_MALE_Base_Skeleton
      SkinningMode=SkinOnTheFly
      SoftSourceActor=null
      SourceMode=Default
      UvSetIndex=0
      WholeMeshLOD=-1
      bExcludeBone=false
      bReadDeformedGeometry=true
      bRequireCurrentFrameData=true
      ComponentTags []
      FilteredBones [foot_l,calf_l,thigh_l]
      FilteredSockets []
      MeshUserParameter {parameter {name=None, typeDefHandle {registeredTypeIndex=1}, varData []}}
      SamplingRegions []
     }}
    Stabalize Twisting=true
    Write Position Attribute=true
   ) {module=FN_BoneFlow001}
   /SetVariables(Particles.BoneFlowTangent:Vector3f=$Output.FN_BoneFlow001.SplineTangent) {isSetParameters=true, module=SetVariables_#1}
   /SetVariables(Particles.RingRotation:Quat4f=&MultiplyQuaternion(
     Quaternion A=&MakeQuatFromAxisAngle(Axis=$Particles.BoneFlowTangent, AngleInDegrees=&RandomRangeFloat(Minimum=0, Maximum=360))
     Quaternion B=&FindQuatBetweenTwoVectors(From Vector=<0.4 0.4 1>, To Vector=$Particles.BoneFlowTangent)
    )) {isSetParameters=true, module=SetVariables_#2}
   /SetVariables(Particles.ScaleMeshShapeCurve=DataInterface() {propertyValues {Curve {keys [<at 0.1 0.5>,<at 0.2 0.7>,<at 0.7 1>,<at 1 1.2>]}}}) {isSetParameters=true, module=SetVariables_#3}
   /SetVariables(Particles.ScaleMeshXYCurve=DataInterface() {propertyValues {Curve {keys [
        {interpMode=RCIM_Cubic, tangentMode=RCTM_User, time=0, value=0}
        {interpMode=RCIM_Cubic, tangentMode=RCTM_User, time=0.35, value=0.7}
        {arriveTangent=-0.2106, interpMode=RCIM_Cubic, leaveTangent=-0.2106, tangentMode=RCTM_User, time=1, value=1.2}
       ]}}}) {isSetParameters=true, module=SetVariables_#4}
   &ScaleMeshSize(Scale Factor=&Multiply_Vector(
     A=&MakeVector(
      X=&FloatFromCurve(FloatCurve=$Particles.ScaleMeshXYCurve)
      Y=&FloatFromCurve(FloatCurve=$Particles.ScaleMeshXYCurve)
      Z=&FloatFromCurve(FloatCurve=DataInterface() {propertyValues {Curve {keys [<at 0 1>, {interpMode=RCIM_Cubic, tangentMode=RCTM_User, time=0.6, value=1}, {interpMode=RCIM_Cubic, tangentMode=RCTM_User, time=1, value=0}]}}})
     )
     B=&MakeVector(
      X=&FloatFromCurve(FloatCurve=$Particles.ScaleMeshShapeCurve, CurveIndex=$Particles.Normalized Spawn Group)
      Y=&FloatFromCurve(FloatCurve=$Particles.ScaleMeshShapeCurve, CurveIndex=$Particles.Normalized Spawn Group)
      Z=1
     )
    ))
   &DynamicMaterialParameters(
    Index 0 Param 1=0.5
    Index 0 Param 2=&FloatFromCurve(FloatCurve=DataInterface() {propertyValues {Curve {keys [
         {interpMode=RCIM_Cubic, tangentMode=RCTM_User, time=0, value=0.5}
         {arriveTangent=2.7221, interpMode=RCIM_Cubic, leaveTangent=2.7221, tangentMode=RCTM_User, time=0.5, value=1.5}
         {arriveTangent=0.4617, interpMode=RCIM_Cubic, leaveTangent=0.4617, tangentMode=RCTM_User, time=0.95, value=2}
         {interpMode=RCIM_Cubic, tangentMode=RCTM_User, time=1, value=2.5}
        ]}}})
    Index 1 Param 1=&FloatFromCurve(FloatCurve=DataInterface() {propertyValues {Curve {keys [<at 0 2>,<at 0.4 0.5>]}}})
    Index 1 Param 2=2
    Index 1 Param 3=&FloatFromCurve(FloatCurve=DataInterface() {propertyValues {Curve {keys [<at 0 10>,<at 0.3 1>]}}})
    Module.Param0WriteEnabled=true
    Module.Param0WriteEnabled1=true
    Module.Param1WriteEnabled=true
    Module.Param1WriteEnabled1=true
    Module.Param2WriteEnabled=true
    Module.Param2WriteEnabled1=true
    Module.Param3WriteEnabled=true
    Module.Param3WriteEnabled1=true
    Write Parameter Index 1=true
   )
  ]
  rendererClasses [&NiagaraMeshRendererProperties]
  renderers [{rendererClass=&NiagaraMeshRendererProperties, rendererIndex=0}]
 }
 BoneFlow_Leg_R {
  emitterUpdateScript [
   &EmitterState(Life Cycle Mode=::NewEnumerator1, Loop Behavior=::NewEnumerator1, Loop Duration=0.4, UseLoopDelay=false)
   &SpawnRate(SpawnRate=50, Use Spawn Probability=false)
  ]
  particleSpawnScript [
   &FN_SetSpawnGroup(Inclusivity=true, Normalize Spawn Group=true, Particles Per Group=20, Pattern=::NewEnumerator0, Set Ribbon ID and Ribbon Link Order=true) {module=FN_SetSpawnGroup001}
   &InitializeParticle(
    Color=<rgba 0.9249 1 0.9295 1>
    Lifetime=&FloatFromCurve(
     FloatCurve=DataInterface() {propertyValues {Curve {keys [<at 0 0.3>,<at 0.7 0.6>,<at 1 1.2>]}}}
     CurveIndex=&RandomRangeFloat(Minimum=0, Maximum=1)
    )
    Mesh Scale Max=<1.4 1.4 3>
    Mesh Scale Min=<0.9 0.9 2>
    Mesh Scale Mode=::NewEnumerator2
    UsePositionOffset=false
    Write Material Random=true
   )
   &InitialMeshOrientation(Mesh Orientation Mode=::NewEnumerator2, Use Rotation Vector=true)
   &CameraOffset(Camera Offset Amount=10)
  ]
  particleUpdateScript [
   &ParticleState()
   &FN_BoneFlow(
    Angle:NiagaraFloat=&Multiply_Float(A=$Particles.Normalized Spawn Group, B=-1)
    BoneID:NiagaraInt32=&ModuloInt(A=$Particles.SpawnGroup, B=1)
    Height:NiagaraFloat=&Multiply_Float(A=$Particles.Normalized Spawn Group, B=1.2)
    Interpolate Mesh Sampling=false
    Radius:NiagaraFloat=&Multiply_Float(
     A=&FloatFromCurve(
      FloatCurve=DataInterface() {propertyValues {Curve {keys [
          {arriveTangent=64.149, interpMode=RCIM_Cubic, leaveTangent=64.149, tangentMode=RCTM_User, time=0, value=0}
          <at 0.5 10 RCIM_Cubic>
          {arriveTangent=-53.2228, interpMode=RCIM_Cubic, leaveTangent=-53.2228, tangentMode=RCTM_User, time=1, value=0}
         ]}}}
      CurveIndex=$Particles.Normalized Spawn Group
     )
     B=1
    )
    SideVector2:Vector3f=<0 0 0>
    SideVector3:Vector3f=<1 1 1>
    SkeletalMesh=DataInterface() {propertyValues {
      DefaultMesh=null
      ExcludeBoneName=root
      OptionalComponentClass=null
      PreviewMesh=&M_MED_Clash_V_Casual
      SkinningMode=SkinOnTheFly
      SoftSourceActor=null
      SourceMode=Default
      UvSetIndex=0
      WholeMeshLOD=-1
      bExcludeBone=false
      bReadDeformedGeometry=true
      bRequireCurrentFrameData=true
      ComponentTags []
      FilteredBones [foot_r,calf_r,thigh_r]
      FilteredSockets []
      MeshUserParameter {parameter {name=None, typeDefHandle {registeredTypeIndex=1}, varData []}}
      SamplingRegions []
     }}
    Stabalize Twisting=true
    Write Position Attribute=true
   ) {module=FN_BoneFlow001}
   /SetVariables(Particles.BoneFlowTangent:Vector3f=$Output.FN_BoneFlow001.SplineTangent) {isSetParameters=true, module=SetVariables_#1}
   /SetVariables(Particles.RingRotation:Quat4f=&MultiplyQuaternion(
     Quaternion A=&MakeQuatFromAxisAngle(Axis=$Particles.BoneFlowTangent, AngleInDegrees=&RandomRangeFloat(Minimum=0, Maximum=360))
     Quaternion B=&FindQuatBetweenTwoVectors(From Vector=<0.4 0.4 1>, To Vector=$Particles.BoneFlowTangent)
    )) {isSetParameters=true, module=SetVariables_#2}
   /SetVariables(Particles.ScaleMeshShapeCurve=DataInterface() {propertyValues {Curve {keys [<at 0.1 0.5>,<at 0.2 0.7>,<at 0.7 1>,<at 1 1.2>]}}}) {isSetParameters=true, module=SetVariables_#3}
   /SetVariables(Particles.ScaleMeshXYCurve=DataInterface() {propertyValues {Curve {keys [
        {interpMode=RCIM_Cubic, tangentMode=RCTM_User, time=0, value=0}
        {interpMode=RCIM_Cubic, tangentMode=RCTM_User, time=0.35, value=0.7}
        {arriveTangent=-0.2106, interpMode=RCIM_Cubic, leaveTangent=-0.2106, tangentMode=RCTM_User, time=1, value=1.2}
       ]}}}) {isSetParameters=true, module=SetVariables_#4}
   &ScaleMeshSize(Scale Factor=&Multiply_Vector(
     A=&MakeVector(
      X=&FloatFromCurve(FloatCurve=$Particles.ScaleMeshXYCurve)
      Y=&FloatFromCurve(FloatCurve=$Particles.ScaleMeshXYCurve)
      Z=&FloatFromCurve(FloatCurve=DataInterface() {propertyValues {Curve {keys [<at 0 1>, {interpMode=RCIM_Cubic, tangentMode=RCTM_User, time=0.6, value=1}, {interpMode=RCIM_Cubic, tangentMode=RCTM_User, time=1, value=0}]}}})
     )
     B=&MakeVector(
      X=&FloatFromCurve(FloatCurve=$Particles.ScaleMeshShapeCurve, CurveIndex=$Particles.Normalized Spawn Group)
      Y=&FloatFromCurve(FloatCurve=$Particles.ScaleMeshShapeCurve, CurveIndex=$Particles.Normalized Spawn Group)
      Z=1
     )
    ))
   &DynamicMaterialParameters(
    Index 0 Param 1=0.5
    Index 0 Param 2=&FloatFromCurve(FloatCurve=DataInterface() {propertyValues {Curve {keys [
         {interpMode=RCIM_Cubic, tangentMode=RCTM_User, time=0, value=0.5}
         {arriveTangent=2.7221, interpMode=RCIM_Cubic, leaveTangent=2.7221, tangentMode=RCTM_User, time=0.5, value=1.5}
         {arriveTangent=0.4617, interpMode=RCIM_Cubic, leaveTangent=0.4617, tangentMode=RCTM_User, time=0.95, value=2}
         {interpMode=RCIM_Cubic, tangentMode=RCTM_User, time=1, value=2.5}
        ]}}})
    Index 1 Param 1=&FloatFromCurve(FloatCurve=DataInterface() {propertyValues {Curve {keys [<at 0 2>,<at 0.4 0.5>]}}})
    Index 1 Param 2=2
    Index 1 Param 3=&FloatFromCurve(FloatCurve=DataInterface() {propertyValues {Curve {keys [<at 0 10>,<at 0.3 1>]}}})
    Module.Param0WriteEnabled=true
    Module.Param0WriteEnabled1=true
    Module.Param1WriteEnabled=true
    Module.Param1WriteEnabled1=true
    Module.Param2WriteEnabled=true
    Module.Param2WriteEnabled1=true
    Module.Param3WriteEnabled=true
    Module.Param3WriteEnabled1=true
    Write Parameter Index 1=true
   )
  ]
  rendererClasses [&NiagaraMeshRendererProperties]
  renderers [{rendererClass=&NiagaraMeshRendererProperties, rendererIndex=0}]
 }
 Mesh_Calf {
  emitterUpdateScript [
   &EmitterState(Inactive Response=::NewEnumerator1, Life Cycle Mode=::NewEnumerator1, Loop Behavior=::NewEnumerator1, Loop Duration=4, UseLoopDelay=false)
   &SpawnBurst_Instantaneous(Spawn Count=2, Use Spawn Probability=false, UseLoopCountLimit=false)
  ]
  particleSpawnScript [
   &FN_SetSpawnGroup(Number of Spawn Groups=2)
   &InitializeParticle(
    Lifetime Max=1
    Lifetime Min=0.8
    Lifetime Mode=::NewEnumerator1
    Mesh Scale=&RandomRangeVector(Minimum=<1.2 0.6 0.6>, Maximum=<1 0.7 0.7>)
    Mesh Scale Mode=::NewEnumerator4
    UsePositionOffset=false
    Write Material Random=true
   )
   &InitialMeshOrientation(
    Orientation Axis=&SelectVectorFromArray(Direct Array Index=$Particles.SpawnGroup, Vector Selection Array=DataInterface() {propertyValues {FloatData [<-1 0 0>, <1 0 0>]}})
    Rotation=&RandomRangeVector(Minimum=<1 0 0>, Maximum=<0 0 0>)
    Rotation Coordinate Space=::NewEnumerator2
    Use Orientation Vector=true
    Use Rotation Vector=true
   )
   &CameraOffset()
  ]
  particleUpdateScript [
   &ParticleState()
   &UpdateMeshOrientation(Rotation Rate=-0.1, Rotation Vector=<1 0 0>)
   &SkeletalMeshLocation(
    Bone / Socket Index=$Particles.SpawnGroup
    Bone Sampling Mode=::NewEnumerator4
    "Interpolate Mesh Sampling (Update)"=true
    MaskSpawnGroup=false
    Sampled Position Offset=&SelectVectorFromArray(
     Direct Array Index=$Particles.SpawnGroup
     Vector Selection Array=DataInterface() {propertyValues {MaxElements=2, FloatData [<-18 0 5>, <18 0 -5>]}}
    )
    Skeletal Mesh=DataInterface() {propertyValues {
      DefaultMesh=null
      ExcludeBoneName=root
      OptionalComponentClass=null
      PreviewMesh=&M_MED_Clash_V_Casual
      SkinningMode=SkinOnTheFly
      SoftSourceActor=null
      SourceMode=Default
      UvSetIndex=0
      WholeMeshLOD=-1
      bExcludeBone=false
      bReadDeformedGeometry=true
      bRequireCurrentFrameData=true
      ComponentTags []
      FilteredBones [calf_l,calf_r]
      FilteredSockets []
      MeshUserParameter {parameter {name=None, typeDefHandle {registeredTypeIndex=1}, varData []}}
      SamplingRegions []
     }}
   )
   &DynamicMaterialParameters(
    Index 0 Param 1=&FloatFromCurve(FloatCurve=DataInterface() {propertyValues {Curve {keys [<at 0 1>,<at 0.5 0.5>]}}})
    Index 0 Param 2=&FloatFromCurve(FloatCurve=DataInterface() {propertyValues {Curve {keys [
         {interpMode=RCIM_Cubic, leaveTangent=-2.9591, tangentMode=RCTM_Break, time=0, value=1.5}
         {arriveTangent=-0.3114, interpMode=RCIM_Cubic, tangentMode=RCTM_Break, time=1, value=0}
        ]}}})
    Index 1 Param 1=&FloatFromCurve(FloatCurve=DataInterface() {propertyValues {Curve {keys [
         {arriveTangent=-10.7468, interpMode=RCIM_Cubic, leaveTangent=-10.7468, tangentMode=RCTM_User, time=0, value=2}
         {interpMode=RCIM_Cubic, leaveTangent=0.5305, tangentMode=RCTM_Break, time=0.1, value=1}
         {interpMode=RCIM_Cubic, tangentMode=RCTM_Break, time=1, value=0.2}
        ]}}})
    Index 1 Param 2=2
    Index 1 Param 3=&FloatFromCurve(FloatCurve=DataInterface() {propertyValues {Curve {keys [
         {interpMode=RCIM_Cubic, leaveTangent=10.6169, tangentMode=RCTM_Break, time=0, value=-0.2}
         {arriveTangent=2.4889, interpMode=RCIM_Cubic, leaveTangent=0.1906, tangentMode=RCTM_Break, time=0.1, value=0.5}
         {arriveTangent=-0.3416, interpMode=RCIM_Cubic, leaveTangent=-1.8967, tangentMode=RCTM_Break, time=0.6, value=0.5}
         {arriveTangent=0.3626, interpMode=RCIM_Cubic, tangentMode=RCTM_Break, time=0.9, value=0}
        ]}}})
    Module.Param0WriteEnabled=true
    Module.Param0WriteEnabled1=true
    Module.Param1WriteEnabled=true
    Module.Param1WriteEnabled1=true
    Module.Param2WriteEnabled=false
    Module.Param2WriteEnabled1=true
    Module.Param3WriteEnabled=false
    Module.Param3WriteEnabled1=false
    Write Parameter Index 1=true
   )
   &ScaleMeshSize(Scale Factor=&MakeVector(
     X=&FloatFromCurve(FloatCurve=DataInterface() {propertyValues {Curve {keys [<at 0.65 1>,<at 1 0.8>]}}})
     Y=&FloatFromCurve(FloatCurve=DataInterface() {propertyValues {Curve {keys [
          {interpMode=RCIM_Cubic, tangentMode=RCTM_User, time=0, value=0.7}
          {interpMode=RCIM_Cubic, tangentMode=RCTM_User, time=0.6, value=1}
          {arriveTangent=-4.4791, interpMode=RCIM_Cubic, leaveTangent=-4.4791, tangentMode=RCTM_User, time=1, value=0}
         ]}}})
     Z=&FloatFromCurve(FloatCurve=DataInterface() {propertyValues {Curve {keys [
          {interpMode=RCIM_Cubic, tangentMode=RCTM_User, time=0, value=0.7}
          {interpMode=RCIM_Cubic, tangentMode=RCTM_User, time=0.6, value=1}
          {arriveTangent=-6.148, interpMode=RCIM_Cubic, leaveTangent=-6.148, tangentMode=RCTM_User, time=1, value=0}
         ]}}})
    ))
  ]
  rendererClasses [&NiagaraMeshRendererProperties]
  renderers [{rendererClass=&NiagaraMeshRendererProperties, rendererIndex=0}]
 }
}
summary {
 systemName=NS_Emote_Clash_Removal
 emitters [
  {emitterName=BoneFlow_Leg_L, rendererClasses [&NiagaraMeshRendererProperties]}
  {emitterName=BoneFlow_Leg_R, rendererClasses [&NiagaraMeshRendererProperties]}
  {emitterName=BoneFlow_Body, rendererClasses [&NiagaraMeshRendererProperties]}
  {emitterName=BoneFlow_Arm_L, rendererClasses [&NiagaraMeshRendererProperties]}
  {emitterName=BoneFlow_Arm_R, rendererClasses [&NiagaraMeshRendererProperties]}
  {emitterName=Mesh_Calf, rendererClasses [&NiagaraMeshRendererProperties]}
 ]
 userVariables []
}
